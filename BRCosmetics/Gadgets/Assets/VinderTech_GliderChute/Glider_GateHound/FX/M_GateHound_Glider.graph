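# Translucent, unlit, two-sided FX material for flame-like trails on the GateHound glider, usable on Niagara ribbons and meshes.
# Color combines a curve-atlas gradient, panning noise and particle color.
# Opacity multiplies a noise-distorted scrolling fire texture, caustic-noise erosion and tail masks, with optional Fresnel, view-angle fading and depth fade on deferred shading.
# Dynamic parameters drive scroll, erosion, UV offset, a U-weighted horizontal world-position bend and optional SplineThicken offsets.
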Material /BRCosmetics/Gadgets/Assets/VinderTech_GliderChute/Glider_GateHound/FX/M_GateHound_Glider
graph {
 properties {
  bUsedWithMeshDeformer=true
  bUsedWithNiagaraMeshParticles=true
  bUsedWithNiagaraRibbons=true
  bUsedWithSkeletalMesh=true
  blendMode=BLEND_Translucent
  pixelDepthOffsetMode=PDOM_Legacy
  shadingModel=MSM_Unlit
  twoSided=true
  usedShadingModels=Unlit
 }
 nodes [
  n1 = DynamicParameter() {parameterIndex=1, defaultValue=<rgba 1 1 1 0>, paramNames [X,Y,Z,WPOIntensity]}
  n2 = Clamp(
   None=Subtract(
    A=MaterialFunctionCall(X=@n1.X, Y=@n1.X) {materialFunction=&MakeFloat3}.Result
    B=ParticlePositionWS()
   )
   Min=Constant() {r=-30}
   Max=Constant() {r=30}
  )
  n3 = TextureCoordinate()
  n4 = Multiply(A=@n1.WPOIntensity, B=Power(Base=ComponentMask(None=@n3) {r=true}, Exp=ScalarParameter() {defaultValue=2, parameterName=WPOBlendPower}))
  LinearInterpolate(
   B=Multiply(A=LinearInterpolate(B=@n2, Alpha=@n4), B=Constant3Vector() {constant=<rgba 1 1 0 0>})
   Alpha=ScalarParameter() {defaultValue=1, parameterName=Use Interpolate WPO}
  )
  n5 = TextureCoordinate() {uTiling=2}
  n6 = DynamicParameter() {paramNames [Erosion, Speed, WPO Disturbance, U offset]}
  n7 = AppendVector(A=@n6.U offset, B=Constant())
  n8 = TextureSample(UVs=Panner(Coordinate=Add(A=@n5, B=@n7), Time=@n6.Speed) {speedX=-1}) {samplerType=SAMPLERTYPE_LinearGrayscale, texture=/Engine/Functions/Engine_MaterialFunctions02/ExampleContent/Textures/LowResBlurredNoise}
  n9 = LinearInterpolate(Alpha=@n8.R) {constA=-0.8}
  n10 = ScalarParameter() {defaultValue=0.2, group="🎨Color", parameterName=Dark noise min}
  n11 = CurveAtlasRowParameter(CurveTime=ComponentMask(None=TextureCoordinate()) {r=true}) {
   atlas=/BRCosmetics/Gadgets/Assets/VinderTech_GliderChute/Glider_GateHound/CAtlas_GateHound
   curve=/BRCosmetics/Gadgets/Assets/VinderTech_GliderChute/Glider_GateHound/FX/Curves/CC_GateHound_1
   parameterName=Color
  }
  n12 = Multiply(
   A=LinearInterpolate(B=Clamp(None=@n9, Min=@n10) {minDefault=0.1}, Alpha=ScalarParameter() {defaultValue=1, group="🎨Color", parameterName=Add Noise}) {constA=1}
   B=@n11
  )
  n13 = ScalarParameter() {defaultValue=1, group="🎨Color", parameterName=Use Color Curve}
  n14 = ParticleColor()
  MaterialFunctionCall(Emissive=Multiply(A=LinearInterpolate(B=@n12, Alpha=@n13) {constA=1}, B=@n14.RGB)) {materialFunction=&MF_FX_EmissiveAdjustments}
  n15 = TextureCoordinate()
  n16 = ComponentMask(None=MaterialFunctionCall(
    X=ScalarParameter() {defaultValue=0.7, group="🔻CORE SHAPE", parameterName=Main_ScaleX}
    Y=ScalarParameter() {defaultValue=1, group="🔻CORE SHAPE", parameterName=Main_ScaleY}
   ) {materialFunction=&MakeFloat2}.Result) {g=true, r=true}
  n17 = TextureSample(UVs=Panner(Coordinate=TextureCoordinate()) {speedX=-0.3, speedY=0.3}) {samplerType=SAMPLERTYPE_Grayscale, texture=/Game/Global/Textures/Default/Noise/LowResBlurredNoise}
  n18 = ScalarParameter() {group="🔻CORE SHAPE", parameterName=Noise intensity}
  n19 = Add(
   A=Add(A=Multiply(A=@n15, B=@n16), B=Multiply(A=@n17.R, B=@n18) {constB=0.2})
   B=AppendVector(A=DynamicParameter() {paramNames [Erosion, Speed, WPO Disturbance, U offset]}.U offset, B=Constant())
  )
  n20 = DynamicParameter() {paramNames [Erosion, Speed, WPO Disturbance, U offset]}
  n21 = ChannelMaskParameter(None=TextureSampleParameter2D(UVs=Panner(Coordinate=@n19, Time=@n20.Speed) {speedX=-0.3}) {group=Shape, parameterName=MainTex, texture=/BRCosmetics/Weapons/FORT_Melee/Pickaxe_MadameMoth/FX/T_MadameMoth_Fire_Trail01}.RGBA) {group=Tex, parameterName=Mask}
  n22 = ScalarParameter() {defaultValue=1, group="🔻CORE SHAPE", parameterName=Main_Power}
  n23 = TextureCoordinate()
  n24 = Multiply(A=OneMinus(None=ComponentMask(None=@n23) {r=true})) {constB=5}
  n25 = Multiply(A=Multiply(A=Power(Base=@n21.R, Exp=@n22) {constExponent=2.2}, B=Saturate(None=@n24)), B=@n14.A)
  n26 = DynamicParameter() {paramNames [Erosion, Speed, WPO Disturbance, U offset]}
  n27 = ScalarParameter() {defaultValue=-3, group="🔥EROSION", parameterName=Disappear min}
  n28 = Multiply(
   A=TextureCoordinate() {uTiling=2}
   B=MaterialFunctionCall(
    X=ScalarParameter() {defaultValue=1, group="🔥EROSION", parameterName=Erosion_ScaleX}
    Y=ScalarParameter() {defaultValue=1, group="🔥EROSION", parameterName=Erosion_ScaleY}
   ) {materialFunction=&MakeFloat2}.Result
  )
  n29 = DynamicParameter() {paramNames [Erosion, Speed, WPO Disturbance, U offset]}
  n30 = AppendVector(A=@n29.U offset, B=Constant())
  n31 = TextureSample(UVs=Panner(Coordinate=Add(A=@n28, B=@n30), Time=@n29.Speed) {speedX=-0.8}) {samplerType=SAMPLERTYPE_LinearGrayscale, texture=/Game/Global/Textures/Default/Caustics/T_Nosie_Caustics_Grayscale}
  n32 = ComponentMask(None=TextureCoordinate()) {r=true}
  n33 = Saturate(None=LinearInterpolate(
    A=ScalarParameter() {group="🔥EROSION", parameterName=Dissolve mask Min}
    B=ScalarParameter() {defaultValue=0.5, group="🔥EROSION", parameterName=Dissolve mask Max}
    Alpha=Power(Base=@n32) {constExponent=1.3}
   ) {constA=0.1, constB=2})
  n34 = Saturate(None=LinearInterpolate(
    A=Multiply(A=@n26.Erosion, B=@n27)
    B=ScalarParameter() {defaultValue=2, group="🔥EROSION", parameterName=Disappear max}
    Alpha=Power(Base=@n31.R, Exp=@n33) {constExponent=0.2}
   ) {constA=-2, constB=2})
  n35 = OneMinus(None=ComponentMask(None=TextureCoordinate()) {r=true})
  n36 = ScalarParameter() {defaultValue=3, parameterName=Tail Power}
  n37 = Saturate(None=Multiply(A=Power(Base=@n35, Exp=@n36) {constExponent=4}, B=ScalarParameter() {defaultValue=8, parameterName=Tail edge hardness}) {constB=5})
  n38 = TextureCoordinate()
  n39 = Constant2Vector() {g=0.5, r=0.7}
  n40 = Saturate(None=Multiply(A=SphereMask(A=@n38, B=@n39) {attenuationRadius=0.7, hardnessPercent=0}, B=ScalarParameter() {defaultValue=20, parameterName=B}) {constB=20})
  n41 = LinearInterpolate(
   A=Multiply(A=Multiply(A=@n25, B=@n34), B=Multiply(A=@n37, B=@n40))
   B=Multiply(A=Multiply(A=@n14.A, B=@n34), B=@n37)
   Alpha=ScalarParameter() {group="🔍MISC", parameterName=WideFlames?}
  )
  n42 = ScalarParameter() {defaultValue=1, group="🔍MISC", parameterName="Brightness (S)"}
  n43 = ScalarParameter() {defaultValue=3, group="🔍MISC", parameterName=ExponentIn}
  n44 = LinearInterpolate(
   A=@n41
   B=Multiply(
    A=@n41
    B=MaterialFunctionCall(Brightness=@n42, Exponent=@n43) {materialFunction=&MF_Fresnel_FX}.Fresnel
   )
   Alpha=ScalarParameter() {parameterName=FresnelFade?}
  )
  n45 = DepthFade(Opacity=@n44, FadeDistance=ScalarParameter() {defaultValue=20, group="🔍MISC", parameterName=FadeDistance})
  n46 = ShadingPathSwitch(Default=@n45, Deferred=@n45, Forward=@n44, Mobile=@n44)
  n47 = MaterialFunctionCall("FadeSensitivity (optional)"=ScalarParameter() {group="🔍MISC", parameterName="FadeSensitivity (optional) (S)"}) {materialFunction=&FadeBasedOnViewAngle}
  Multiply(
   A=Multiply(A=Multiply(A=@n46, B=@n47.Result), B=ScalarParameter() {defaultValue=1, group="🔍MISC", parameterName=GlobalAlpha})
   B=VertexColor()
  )
  n48 = DynamicParameter() {parameterIndex=2, paramNames [SplineBase, SplineTip, WPO Disturbance, U offset]}
  n49 = MaterialFunctionCall(WidthBase=@n48.SplineTip, WidthTip=@n48.SplineTip, Expand U or V UV Channel=StaticBool()) {materialFunction=&SplineThicken}
  n50 = Constant()
  Reroute(None=Add(A=StaticSwitchParameter(True=@n49.WorldPositionOffset, False=@n50) {parameterName=UseSplineThicken}) {constB=0})
 ]
}
