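# Unlit, translucent, two-sided material for streak effects on the GateHound glider, enabled for Niagara mesh particles.
# A panning fire-noise texture is eroded, overlay-blended with a remapped copy of itself, then remapped between Min and Max.
# The result is multiplied by particle alpha and a powered, smoothstepped V-axis mask that fades both edges.
# A separate inverted reflected gradient feeds nothing shown, and no material output connections appear.
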Material /BRCosmetics/Gadgets/Assets/VinderTech_GliderChute/Glider_GateHound/FX/M_Gatehound_Streaks
graph {
 properties {
  bUsedWithNiagaraMeshParticles=true
  blendMode=BLEND_Translucent
  pixelDepthOffsetMode=PDOM_Legacy
  shadingModel=MSM_Unlit
  twoSided=true
  usedShadingModels=Unlit
 }
 nodes [
  n1 = ScalarParameter() {parameterName=Min}
  n2 = ScalarParameter() {defaultValue=0.5, parameterName=Max}
  n3 = TextureSample(UVs=MaterialFunctionCall!(
    UV
    Tile U=ScalarParameter() {defaultValue=1, parameterName="Tile U (S)"}
    Tile V=ScalarParameter() {defaultValue=1, parameterName="Tile V (S)"}
    Pan U=ScalarParameter() {defaultValue=3, parameterName="Pan U (S)"}
    Pan V=ScalarParameter() {defaultValue=1, parameterName="Pan V (S)"}
    Offset U
    Offset V
    Time
   ) {materialFunction=/Game/Effects/Fort_Effects/Materials/_Functions/MF_PanTileOffset}.UV's) {samplerType=SAMPLERTYPE_LinearGrayscale, texture=&/T_Mojo_Fire_Noise}
  n4 = ScalarParameter() {defaultValue=-0.5, parameterName=Param_1}
  n5 = ScalarParameter() {defaultValue=0.5, parameterName=Param_2}
  n6 = MaterialFunctionCall(
   Base=MaterialFunctionCall!(
    Gradient=@n3.R
    Erosion Amount=ScalarParameter() {defaultValue=0.2, parameterName="Erosion Amount (S)"}
    Softness=ScalarParameter() {defaultValue=1, parameterName="Softness (S)"}
   ) {materialFunction=/BRCosmetics/Athena/Effects/Particles/Emotes/NS_Emote_Spectrum/Functions/MF_Erosion}.Default Range
   Blend=Saturate(None=LinearInterpolate(A=@n4, B=@n5, Alpha=@n3.R))
  ) {materialFunction=&Blend_Overlay}
  n7 = ComponentMask(None=TextureCoordinate()) {g=true}
  n8 = OneMinus(None=Abs(None=Subtract(A=@n7) {constB=0.5}))
  n9 = ScalarParameter() {defaultValue=8, parameterName=Exp}
  Multiply(
   A=Multiply(A=ParticleColor().A, B=LinearInterpolate(A=@n1, B=@n2, Alpha=@n6.Result))
   B=SmoothStep(Max=ScalarParameter() {defaultValue=0.025, parameterName=Max_1}, Value=Power(Base=@n8, Exp=@n9)) {constMax=0.2}
  )
  OneMinus(None=MaterialFunctionCall!(UVs, Offset, RotationCenter, RotationAngle=Constant() {r=0.25}) {materialFunction=/FortUILibrary/MaterialFunctions/Gradients/MF_UI_ReflectedGradient}.Result)
 ]
}
