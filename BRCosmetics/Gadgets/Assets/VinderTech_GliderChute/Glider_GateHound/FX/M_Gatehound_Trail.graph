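# Unlit translucent Niagara ribbon material for the Gatehound glider trail.
# A channel-selected, power-adjusted Madame Moth contrail texture is multiplied by particle alpha and a dissolve mask to form opacity.
# The mask comes from HeightLerp over a tiled, axis-swapped cloud texture, driven by the Dissolve dynamic parameter.
# A shading path switch applies depth fade on default and deferred paths, skipping it on forward and mobile.
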Material /BRCosmetics/Gadgets/Assets/VinderTech_GliderChute/Glider_GateHound/FX/M_Gatehound_Trail
graph {
 properties {bUsedWithNiagaraRibbons=true, blendMode=BLEND_Translucent, pixelDepthOffsetMode=PDOM_Legacy, shadingModel=MSM_Unlit, usedShadingModels=Unlit}
 nodes [
  ParticleColor()
  n1 = Power(Base=TextureSample() {texture=/BRCosmetics/Athena/Effects/Particles/Skydiving/NS_Contrail_MadameMoth/T_MadameMoth_Trail03}.RGBA)
  n2 = TextureCoordinate() {vTiling=0.3}
  n3 = MaterialFunctionCall(
   Transition Phase=DynamicParameter() {defaultValue=<rgba 1 -1 1 1>, paramNames [Dissolve, GradientProgress, Param3, Param4]}.Dissolve
   Height Texture=TextureSample(UVs=MaterialFunctionCall(XY=@n2) {materialFunction=&Swizzle}.YX) {samplerType=SAMPLERTYPE_LinearGrayscale, texture=/Game/Athena/Prototype/Terrain/T_CloudCoverage}.R
  ) {materialFunction=&HeightLerp}
  n4 = Multiply(
   A=Multiply(A=ChannelMaskParameter(None=@n1) {parameterName=MainTexChannel}.R, B=ParticleColor().A)
   B=Saturate(None=Multiply(A=@n3.Alpha) {constB=3})
  )
  n5 = DepthFade(Opacity=@n4)
  ShadingPathSwitch(Default=@n5, Deferred=@n5, Forward=@n4, Mobile=@n4)
 ]
}
