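# Unlit translucent glow material for the GateHound glider's effects, enabled for Niagara mesh particles.
# A tint color parameter, green-cyan by default, is multiplied by particle color.
# Inverted U coordinate runs through a parameterized SmoothStep into a three-value lerp, shaping a gradient along the mesh that is scaled by particle alpha.
# A shading path switch applies depth fade to that alpha on default and deferred paths, skipping it on forward and mobile.
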Material /BRCosmetics/Gadgets/Assets/VinderTech_GliderChute/Glider_GateHound/FX/M_Glow_GateHound
graph {
 properties {
  bUsedWithNiagaraMeshParticles=true
  blendMode=BLEND_Translucent
  pixelDepthOffsetMode=PDOM_Legacy
  shadingModel=MSM_Unlit
  usedShadingModels=Unlit
 }
 nodes [
  n1 = ParticleColor()
  Multiply(A=VectorParameter() {defaultValue=<rgba 0 5 1.8333 1>}.RGB, B=@n1.RGB)
  Multiply()
  n2 = ScalarParameter() {parameterName=Param_4}
  n3 = ScalarParameter() {defaultValue=1, parameterName=Param_3}
  n4 = OneMinus(None=ComponentMask(None=TextureCoordinate()) {r=true})
  n5 = Multiply(
   A=MaterialFunctionCall!(
    1=ScalarParameter() {parameterName="1 (S)"}
    2=ScalarParameter() {defaultValue=0.1, parameterName="2 (S)"}
    3=ScalarParameter() {defaultValue=1, parameterName="3 (S)"}
    4
    A=SmoothStep(Min=@n2, Max=@n3, Value=@n4)
   ) {materialFunction=/Engine/Functions/Engine_MaterialFunctions02/Lerp_Multiple_Float}.Lerp 3 Inputs
   B=@n1.A
  )
  n6 = DepthFade(Opacity=@n5, FadeDistance=ScalarParameter())
  ShadingPathSwitch(Default=@n6, Deferred=@n6, Forward=@n5, Mobile=@n5)
 ]
}
