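# Niagara trail effect for the GateHound glider, with green sparks and glow attached to the glider's skeletal mesh and a blue color variant.
# CPU and GPU spark emitters partition particles across bones or sockets, with wheel emitters throwing sparks along a shape's tangent.
# Glow combines meshes, sprites and lights, tinted by time of day and front-end state.
# System Overshoot and Timeline values drive spawn rates and glow fades; FN_Debug visualizes User.ForwardDot.
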NiagaraSystem /BRCosmetics/Gadgets/Assets/VinderTech_GliderChute/Glider_GateHound/FX/NS_Glider_GateHound_Trail
dependencies {
 usedDataInterfaces [
  &NiagaraDataInterfaceCurve
  &NiagaraDataInterfaceArrayFloat
  &NiagaraDataInterfaceColorCurve
  &NiagaraDataInterfaceSkeletalMesh
 ]
 usedDynamicInputs [&FNTimeOfDayScalar, &MakeCustomFloatFromBool]
 usedModules [
  &self:SystemSpawnScript.SystemScriptSource.SystemScriptGraph.NiagaraNodeAssignment_0.SetVariables_#1
  &self:SystemSpawnScript.SystemScriptSource.SystemScriptGraph.NiagaraNodeAssignment_1.SetVariables_#2
  &self:SystemSpawnScript.SystemScriptSource.SystemScriptGraph.NiagaraNodeAssignment_2.SetVariables_#3
  &FN_VariantManager
  &SystemState
  &self:SystemSpawnScript.SystemScriptSource.SystemScriptGraph.NiagaraNodeAssignment_4.SetVariables_#4
  &FN_Integrator
  &Timeline
  &self:SystemSpawnScript.SystemScriptSource.SystemScriptGraph.NiagaraNodeAssignment_3.SetVariables_#5
 ]
}
emitters {
 CPU_Sparks {
  emitterUpdateScript [
   &EmitterState(Inactive Response=::NewEnumerator1, Life Cycle Mode=::NewEnumerator1, Loop Delay=2.9, Loop Duration=1, UseLoopDelay=false)
   &SpawnRate(SpawnRate=&Lerp_Float(A=0, B=70, Alpha=$System.Overshoot), Use Spawn Probability=false) {enabled=false}
   &SpawnRate(SpawnRate=&Lerp_Float(A=0, B=50, Alpha=$System.Overshoot), Use Spawn Probability=false) {module=SpawnRate001}
  ]
  particleSpawnScript [
   /SetVariables(
    Particles.HighSpeedParticle:NiagaraFloat=&UniformAOrBFloat(A=0, Distribution Weight=0.2)
    Particles.Random size:NiagaraFloat=&RandomRangeFloat(Minimum=0.2, Maximum=0.3)
   ) {isSetParameters=true, module=SetVariables_#6}
   /SetVariables() {isSetParameters=true, module=SetVariables_#7}
   &PartitionParticles(Number Of Partitions=4)
   &InitializeParticle(
    Color=<rgba 30 30 30 1>
    Lifetime=&Lerp_Float(
     A=&RandomRangeFloat(Minimum=0.5, Maximum=1.2)
     B=&RandomRangeFloat(Minimum=0.4, Maximum=0.7)
     Alpha=$Particles.HighSpeedParticle
    )
    Sprite Rotation Mode=::NewEnumerator2
    Sprite Size Max=<3 10>
    Sprite Size Min=<2 10>
    Sprite Size Mode=::NewEnumerator2
    Sprite UV Mode=::NewEnumerator3
    UsePositionOffset=false
    Write Material Random=true
   ) {module=InitializeParticle001}
   &DynamicMaterialParameters(
    Index 0 Param 2=&RandomRangeFloat(Minimum=0, Maximum=4)
    Index 0 Param 4=2.5
    Module.Param0WriteEnabled=true
    Module.Param1WriteEnabled=true
    Module.Param2WriteEnabled=false
    Module.Param3WriteEnabled=true
   ) {module=DynamicMaterialParameters001}
   &AddVelocity(
    Cone Angle=&Lerp_Float(A=20, B=25, Alpha=$Particles.HighSpeedParticle)
    Cone Axis=<0 1 0.8>
    Module.UseConeFalloff=false
    Module.UseVelDistribution=false
    Override Seed=false
    Velocity Mode=::NewEnumerator2
    Velocity Speed=&Lerp_Float(
     A=&RandomRangeFloat(Minimum=150, Maximum=300)
     B=&RandomRangeFloat(Minimum=350, Maximum=400)
     Alpha=$Particles.HighSpeedParticle
    )
   )
   &CameraOffset(Camera Offset Amount=-50) {module=CameraOffset001}
   /SetVariables(Particles.DY3:Vector4f=&MakeVector4(X=3, Y=1, Z=1, W=1)) {isSetParameters=true, module=SetVariables_#8}
  ]
  particleUpdateScript [
   &ParticleState()
   /SetVariables(Particles.Size Scale:NiagaraFloat=&FloatFromCurve(
     FloatCurve=DataInterface() {propertyValues {Curve {keys [
         {interpMode=RCIM_Cubic, tangentMode=RCTM_User, time=0, value=0}
         {interpMode=RCIM_Cubic, tangentMode=RCTM_User, time=0.5, value=0.3}
         {tangentMode=RCTM_User, time=1, value=0}
        ]}}}
     Scale Curve=0.7
    )) {isSetParameters=true, module=SetVariables_#9}
   &SkeletalMeshLocation(
    Bone / Socket Index=$Particles.PartitionParticles.Partition
    MaskSpawnGroup=false
    Mesh Sampling Type=::NewEnumerator5
    Position Sampling=::NewEnumerator1
    Sampled Position Offset=&AddVectorToPosition(Position=&RandomRangeVector(Minimum=<-5 -5 -5>, Maximum=<5 5 5>), Vector=<0 -20 0>)
    Skeletal Mesh=$System.Skeletal Mesh
    Socket Sampling Mode=::NewEnumerator5
    Source Movement Scale:Vector3f=&VectorFromFloat(Value=&Lerp_Float(
      A=&FloatFromCurve(FloatCurve=DataInterface() {propertyValues {Curve {keys [<at 0.7 0>,<at 1 0>]}}})
      B=&FloatFromCurve(FloatCurve=DataInterface() {propertyValues {Curve {keys [<at 0 1>,<at 0.7 1>,<at 1 0>]}}})
      Alpha=$System.Overshoot
     ))
   )
   /Niagara/Modules/Update/Forces/CurlNoiseForce(Noise Frequency=50, Noise Strength=&Lerp_Float(A=100, B=400, Alpha=$Particles.HighSpeedParticle), Pan Noise=true)
   &ScaleSpriteSize(Scale Sprite Size Mode=::NewEnumerator0, Uniform Scale Factor=$Particles.Size Scale) {module=ScaleSpriteSize001}
   &SpriteRotationRate(Rotation Rate=&RandomRangeFloat(Minimum=-5, Maximum=5)) {module=SpriteRotationRate001}
   &Drag(Drag=0.5, Use Linear Drag=true, Use Rotational Drag=true)
   &GravityForce(Gravity=&MakeVector(
     X=0
     Y=0
     Z=&FloatFromCurve(
      FloatCurve=DataInterface() {propertyValues {Curve {keys [
          {arriveTangent=0.215, interpMode=RCIM_Cubic, leaveTangent=0.215, tangentMode=RCTM_User, time=0.3, value=0}
          {arriveTangent=1.9017, interpMode=RCIM_Cubic, leaveTangent=1.9017, tangentMode=RCTM_User, time=1, value=0.5}
         ]}}}
      Scale Curve=980
     )
    ))
   &SolveForcesAndVelocity(Clamp Velocity=false, Limit Acceleration=false)
   &ScaleColor(
    Scale Alpha=&Multiply_Float(
     A=&FloatFromCurve(FloatCurve=DataInterface() {propertyValues {Curve {keys [
          {interpMode=RCIM_Cubic, tangentMode=RCTM_User, time=0, value=0}
          <at 0.2 0>
          {interpMode=RCIM_Cubic, tangentMode=RCTM_User, time=0.5, value=0.75}
          {tangentMode=RCTM_User, time=1, value=0}
         ]}}})
     B=1
    )
    Scale RGB=&Multiply_VectorByFloat(
     Vector=&MakeVectorFromLinearColorRGB(Color=&VariantColorSelector(
       Variant Amount=0
       Original Color=&ColorFromCurve(ColorCurve=DataInterface() {propertyValues {
          AlphaCurve {keys []}
          BlueCurve {keys [<at 0 0.5149 RCIM_Cubic>,<at 0.0079 0.0812>,<at 0.4788 0.0812>,<at 0.8661 0.0167>]}
          GreenCurve {keys [<at 0 1 RCIM_Cubic>,<at 0.0079 0.25>,<at 0.4788 0.25>,<at 0.8661 0.1>]}
          RedCurve {keys [<at 0 0.0297 RCIM_Cubic>,<at 0.0079 0.065>,<at 0.4788 0.065>,<at 0.8661 0>]}
         }})
       Variant 1 Color=&ColorFromCurve(ColorCurve=DataInterface() {propertyValues {
          AlphaCurve {keys []}
          BlueCurve {keys [<at 0.0079 0.25>,<at 0.4788 0.25>,<at 0.8661 0.1>]}
          GreenCurve {keys [<at 0.0079 0.1853>,<at 0.4788 0.1113>,<at 0.8661 0.0267>]}
          RedCurve {keys [<at 0.0079 0.065>,<at 0.4788 0.065>,<at 0.8661 0>]}
         }})
       Variant 2 Color=?
      ))
     Float=&Lerp_Float(A=0.5, B=5, Alpha=$Particles.HighSpeedParticle)
    )
    ScaleA=true
    ScaleRGB=true
   )
   &FN_Sprite_Attributes(
    Color=&MakeLinearColorFromVectorAndFloat(
     "Vector (RGB)"=&MakeVectorFromLinearColorRGB(Color=&MultiplyLinearColor(
       Linear Color=&VariantColorSelector(Variant Amount=0, Original Color=<rgba 0 1.5 0.75 1>, Variant 1 Color=<rgba 0 0.45 1.5 1>, Variant 2 Color=?)
       Scale Factor=&MakeVector4_FromFloat(Float=1)
      ))
     "Float (Alpha)"=&Multiply_Float(A=0.2, B=&FN_FrontEndFloatModifier(Gameplay=0.5, Front End=0.6))
    )
    Create Color Attribute=true
    Create SpriteSize Attribute=true
    SpriteSize=&Multiply_Vector2D(
     A=&Vector2DFromFloat(Value=&Multiply_Float(A=180, B=&Multiply_Float(A=$Particles.Random size, B=$Particles.Size Scale)))
     B=<0.6 1.6>
    )
    Write to Module Namespace=true
   ) {module=FN_Sprite_Attributes001}
  ]
  rendererClasses [&NiagaraSpriteRendererProperties, &NiagaraMeshRendererProperties]
  renderers [
   {rendererClass=&NiagaraSpriteRendererProperties}
   {rendererClass=&NiagaraSpriteRendererProperties, rendererIndex=1}
   {rendererClass=&NiagaraMeshRendererProperties, rendererIndex=2}
  ]
 }
 CPU_Sparks_Wheel {
  emitterUpdateScript [
   &EmitterState(Inactive Response=::NewEnumerator1, Life Cycle Mode=::NewEnumerator1, Loop Delay=2.9, Loop Duration=1, UseLoopDelay=false)
   &SpawnRate(SpawnRate=&Lerp_Float(A=0, B=50, Alpha=$System.Timeline.Time), Use Spawn Probability=false) {module=SpawnRate001}
  ]
  particleSpawnScript [
   &PartitionParticles()
   /SetVariables(
    Particles.HighSpeedParticle:NiagaraFloat=&UniformAOrBFloat(A=0, Distribution Weight=0.01)
    Particles.Random size:NiagaraFloat=&RandomRangeFloat(Minimum=0.2, Maximum=0.3)
   ) {isSetParameters=true, module=SetVariables_#6}
   &InitializeParticle(
    Color=<rgba 10 10 10 1>
    Lifetime=&Lerp_Float(
     A=&RandomRangeFloat(Minimum=0.1, Maximum=0.5)
     B=&RandomRangeFloat(Minimum=0.2, Maximum=0.4)
     Alpha=$Particles.HighSpeedParticle
    )
    Sprite Rotation Mode=::NewEnumerator2
    Sprite Size=&Lerp_Vector2(A=&RandomRangeVector2D(Minimum=<1 15>, Maximum=<5 30>), B=<1 50>, Alpha=$Particles.HighSpeedParticle)
    Sprite Size Mode=::NewEnumerator4
    Sprite UV Mode=::NewEnumerator3
    UsePositionOffset=false
    Write Material Random=true
   ) {module=InitializeParticle001}
   &SkeletalMeshLocation(
    Bone / Socket Index=&Add_Integer(A=4, B=$Particles.PartitionParticles.Partition)
    Bone Sampling Mode=::NewEnumerator4
    MaskSpawnGroup=false
    Sampled Position Offset=&AddVectorToPosition(Position=&RandomRangeVector(Minimum=<-5 -5 -5>, Maximum=<5 5 5>), Vector=<0 0 0>)
    Skeletal Mesh=$System.Skeletal Mesh
   )
   &ShapeLocation(
    Handle Radius=18
    Large Radius=40
    MaskSpawnGroup=false
    Override Seed=false
    Rotation Angle:NiagaraFloat=&Lerp_Float(A=-0.2, B=-0.3, Alpha=&MakeFloatFromInt(Int=$Particles.PartitionParticles.Partition))
    Rotation Mode=::NewEnumerator1
    Shape Primitive=::NewEnumerator3
    U Distribution:NiagaraFloat=&Lerp_Float(A=0.7, B=0.7, Alpha=&MakeFloatFromInt(Int=$Particles.PartitionParticles.Partition))
    UseOwnerScale=true
   )
   &CameraOffset(Camera Offset Amount=-50) {module=CameraOffset001}
   /SetVariables(Particles.DY3:Vector4f=&MakeVector4(X=3, Y=1, Z=1, W=1)) {isSetParameters=true, module=SetVariables_#8}
   /SetVariables(Particles.Velocity=&Multiply_Vector(
     A=$Particles.ShapeLocation.ShapeBitangent
     B=&VectorFromFloat(Value=&Lerp_Float(
       A=&RandomRangeFloat(Minimum=100, Maximum=200)
       B=&RandomRangeFloat(Minimum=1200, Maximum=800)
       Alpha=$Particles.HighSpeedParticle
      ))
    )) {isSetParameters=true, module=SetVariables_#10}
  ]
  particleUpdateScript [
   &ParticleState()
   /SetVariables(Particles.Size Scale:NiagaraFloat=&FloatFromCurve(
     FloatCurve=DataInterface() {propertyValues {Curve {keys [
         {interpMode=RCIM_Cubic, tangentMode=RCTM_User, time=0, value=0}
         {interpMode=RCIM_Cubic, tangentMode=RCTM_User, time=0.5, value=0.3}
         {tangentMode=RCTM_User, time=1, value=0}
        ]}}}
     Scale Curve=0.7
    )) {isSetParameters=true, module=SetVariables_#9}
   &CurlNoiseForce(Noise Frequency=50, Noise Strength=&Lerp_Float(A=100, B=100, Alpha=$Particles.HighSpeedParticle), Pan Noise=true)
   &ScaleSpriteSize(Scale Sprite Size Mode=::NewEnumerator0, Uniform Scale Factor=$Particles.Size Scale) {module=ScaleSpriteSize001}
   &SpriteRotationRate(Rotation Rate=&RandomRangeFloat(Minimum=-5, Maximum=5)) {module=SpriteRotationRate001}
   &Drag(Drag=&Lerp_Float(A=0.5, B=11, Alpha=$Particles.HighSpeedParticle), Use Linear Drag=true, Use Rotational Drag=true)
   &GravityForce(Gravity=&MakeVector(
     X=0
     Y=0
     Z=&FloatFromCurve(
      FloatCurve=DataInterface() {propertyValues {Curve {keys [
          {arriveTangent=0.215, interpMode=RCIM_Cubic, leaveTangent=0.215, tangentMode=RCTM_User, time=0.3, value=0}
          {arriveTangent=1.9017, interpMode=RCIM_Cubic, leaveTangent=1.9017, tangentMode=RCTM_User, time=1, value=0.5}
         ]}}}
      Scale Curve=980
     )
    ))
   &VortexForce(
    UseOriginOffset=false
    Vortex Axis=&RotateVectorByQuaternion(Vector To Rotate=<0 0 1>, Quaternion=$Particles.Initial.SkeletalMeshLocation.SampledRotation)
    Vortex Axis Coordinate Space:ENiagaraCoordinateSpace=::Local
    Vortex Force Amount=&Lerp_Float(A=0, B=400, Alpha=$Particles.HighSpeedParticle)
    Vortex Origin:NiagaraPosition=$Particles.Initial.SkeletalMeshLocation.SampledPosition
   ) {enabled=false}
   &SolveForcesAndVelocity(Clamp Velocity=false, Limit Acceleration=false)
   &ScaleColor(
    Scale Alpha=&Multiply_Float(
     A=&FloatFromCurve(FloatCurve=DataInterface() {propertyValues {Curve {keys [
          {interpMode=RCIM_Cubic, tangentMode=RCTM_User, time=0, value=0}
          <at 0.2 0>
          {interpMode=RCIM_Cubic, tangentMode=RCTM_User, time=0.5, value=0.75}
          {tangentMode=RCTM_User, time=1, value=0}
         ]}}})
     B=1
    )
    Scale RGB=&Multiply_VectorByFloat(
     Vector=&MakeVectorFromLinearColorRGB(Color=&ColorFromCurve(ColorCurve=DataInterface() {propertyValues {
         AlphaCurve {keys []}
         BlueCurve {keys [<at 0 0.5149 RCIM_Cubic>,<at 0.0079 0.1452>,<at 0.4788 0.1452>,<at 0.8661 0.0167>]}
         GreenCurve {keys [<at 0 1 RCIM_Cubic>,<at 0.0079 0.25>,<at 0.4788 0.25>,<at 0.8661 0.1>]}
         RedCurve {keys [<at 0 0.0297 RCIM_Cubic>,<at 0.0079 0.065>,<at 0.4788 0.065>,<at 0.8661 0>]}
        }}))
     Float=&Lerp_Float(A=1, B=4, Alpha=$Particles.HighSpeedParticle)
    )
    ScaleA=true
    ScaleRGB=true
   )
   &FN_Sprite_Attributes(
    Color=&MakeLinearColorFromVectorAndFloat(
     "Vector (RGB)"=&MakeVectorFromLinearColorRGB(Color=&MultiplyLinearColor(
       Linear Color=<rgba 0 1.5 0.675 1>
       Scale Factor=&MakeVector4_FromFloat(Float=&Lerp_Float(A=1, B=2, Alpha=$Particles.HighSpeedParticle))
      ))
     "Float (Alpha)"=&Multiply_Float(A=0.2, B=&FN_FrontEndFloatModifier(Gameplay=0.5, Front End=0.6))
    )
    Create Color Attribute=true
    Create SpriteSize Attribute=true
    SpriteSize=&Multiply_Vector2D(
     A=&Vector2DFromFloat(Value=&Multiply_Float(A=180, B=&Multiply_Float(A=$Particles.Random size, B=$Particles.Size Scale)))
     B=<0.6 1.6>
    )
    Write to Module Namespace=true
   ) {module=FN_Sprite_Attributes001}
  ]
  rendererClasses [&NiagaraSpriteRendererProperties, &NiagaraMeshRendererProperties]
  renderers [
   {rendererClass=&NiagaraSpriteRendererProperties}
   {rendererClass=&NiagaraSpriteRendererProperties, rendererIndex=1}
   {rendererClass=&NiagaraMeshRendererProperties, rendererIndex=2}
  ]
 }
 FN_Debug {
  emitterUpdateScript [
   &EmitterState()
   &SpawnBurst_Instantaneous(Spawn Count=1, Use Spawn Probability=false)
  ]
  particleSpawnScript [/Niagara/Modules/Spawn/Initialization/InitializeParticle(
    Lifetime=5
    Module.Write Color=true
    Module.Write Lifetime=true
    Module.Write Mass=true
    Module.Write Material Random=true
    Module.Write Position=true
    Module.Write Scale=false
    Module.Write SpriteRotation=false
    Module.Write SpriteSize=true
    Position=&AddVectorToPosition(Position=<80 0 0>, Vector=<0 0 0>)
    Sprite Size=<100 100>
   )]
  particleUpdateScript [
   &ParticleState()
   /SetVariables(Particles.Debug:NiagaraFloat=$User.ForwardDot) {isSetParameters=true}
   &DynamicMaterialParameters(
    Index 0 Param 1=$Particles.Debug
    Module.Param0WriteEnabled=true
    Module.Param1WriteEnabled=true
    Module.Param2WriteEnabled=true
    Module.Param3WriteEnabled=true
   )
  ]
  rendererClasses [&NiagaraSpriteRendererProperties]
  renderers [{rendererClass=&NiagaraSpriteRendererProperties}]
 }
 GPU_Sparks {
  simTarget=GPUComputeSim
  emitterUpdateScript [
   &EmitterState(Inactive Response=::NewEnumerator1, Life Cycle Mode=::NewEnumerator1, Loop Delay=2.9, Loop Duration=1, UseLoopDelay=false)
   &SpawnRate(SpawnRate=&Lerp_Float(A=0, B=70, Alpha=$System.Overshoot), Use Spawn Probability=false) {enabled=false}
   &SpawnRate(SpawnRate=&Lerp_Float(A=0, B=70, Alpha=$System.Overshoot), Use Spawn Probability=false) {module=SpawnRate001}
  ]
  particleSpawnScript [
   /SetVariables(
    Particles.HighSpeedParticle:NiagaraFloat=&UniformAOrBFloat(A=0, Distribution Weight=0.2)
    Particles.Random size:NiagaraFloat=&RandomRangeFloat(Minimum=0.2, Maximum=0.3)
   ) {isSetParameters=true, module=SetVariables_#6}
   /SetVariables() {isSetParameters=true, module=SetVariables_#7}
   &PartitionParticles(Number Of Partitions=4)
   /Niagara/Modules/Spawn/Initialization/V2/InitializeParticle(
    Color=<rgba 30 30 30 1>
    Lifetime=&Lerp_Float(
     A=&RandomRangeFloat(Minimum=0.5, Maximum=1.2)
     B=&RandomRangeFloat(Minimum=0.4, Maximum=0.7)
     Alpha=$Particles.HighSpeedParticle
    )
    Sprite Rotation Mode=::NewEnumerator2
    Sprite Size Max=<3 10>
    Sprite Size Min=<2 10>
    Sprite Size Mode=::NewEnumerator2
    Sprite UV Mode=::NewEnumerator3
    UsePositionOffset=false
    Write Material Random=true
   ) {module=InitializeParticle001}
   &DynamicMaterialParameters(
    Index 0 Param 2=&RandomRangeFloat(Minimum=0, Maximum=4)
    Index 0 Param 4=2.5
    Module.Param0WriteEnabled=true
    Module.Param1WriteEnabled=true
    Module.Param2WriteEnabled=false
    Module.Param3WriteEnabled=true
   ) {module=DynamicMaterialParameters001}
   &AddVelocity(
    Cone Angle=&Lerp_Float(A=20, B=25, Alpha=$Particles.HighSpeedParticle)
    Cone Axis=<0 1 0.8>
    Module.UseConeFalloff=false
    Module.UseVelDistribution=false
    Override Seed=false
    Velocity Mode=::NewEnumerator2
    Velocity Speed=&Lerp_Float(
     A=&RandomRangeFloat(Minimum=150, Maximum=300)
     B=&RandomRangeFloat(Minimum=350, Maximum=400)
     Alpha=$Particles.HighSpeedParticle
    )
   )
   &CameraOffset(Camera Offset Amount=-50) {module=CameraOffset001}
   /SetVariables(Particles.DY3:Vector4f=&MakeVector4(X=3, Y=1, Z=1, W=1)) {isSetParameters=true, module=SetVariables_#8}
  ]
  particleUpdateScript [
   &ParticleState()
   /SetVariables(Particles.Size Scale:NiagaraFloat=&FloatFromCurve(
     FloatCurve=DataInterface() {propertyValues {Curve {keys [
         {interpMode=RCIM_Cubic, tangentMode=RCTM_User, time=0, value=0}
         {interpMode=RCIM_Cubic, tangentMode=RCTM_User, time=0.5, value=0.3}
         {tangentMode=RCTM_User, time=1, value=0}
        ]}}}
     Scale Curve=0.7
    )) {isSetParameters=true, module=SetVariables_#9}
   &SkeletalMeshLocation(
    Bone / Socket Index=$Particles.PartitionParticles.Partition
    MaskSpawnGroup=false
    Mesh Sampling Type=::NewEnumerator5
    Position Sampling=::NewEnumerator1
    Sampled Position Offset=&AddVectorToPosition(Position=&RandomRangeVector(Minimum=<-5 -5 -5>, Maximum=<5 5 5>), Vector=<0 -20 0>)
    Skeletal Mesh=$System.Skeletal Mesh
    Socket Sampling Mode=::NewEnumerator5
    Source Movement Scale:Vector3f=&VectorFromFloat(Value=&Lerp_Float(
      A=&FloatFromCurve(FloatCurve=DataInterface() {propertyValues {Curve {keys [<at 0.7 0>,<at 1 0>]}}})
      B=&FloatFromCurve(FloatCurve=DataInterface() {propertyValues {Curve {keys [<at 0 1>,<at 0.3 0>,<at 1 0>]}}})
      Alpha=$System.Overshoot
     ))
   )
   &CurlNoiseForce(Noise Frequency=50, Noise Strength=&Lerp_Float(A=100, B=400, Alpha=$Particles.HighSpeedParticle), Pan Noise=true)
   &ScaleSpriteSize(Scale Sprite Size Mode=::NewEnumerator0, Uniform Scale Factor=$Particles.Size Scale) {module=ScaleSpriteSize001}
   &SpriteRotationRate(Rotation Rate=&RandomRangeFloat(Minimum=-5, Maximum=5)) {module=SpriteRotationRate001}
   &Drag(Drag=0.5, Use Linear Drag=true, Use Rotational Drag=true)
   &GravityForce(Gravity=&MakeVector(
     X=0
     Y=0
     Z=&FloatFromCurve(
      FloatCurve=DataInterface() {propertyValues {Curve {keys [
          {arriveTangent=0.215, interpMode=RCIM_Cubic, leaveTangent=0.215, tangentMode=RCTM_User, time=0.3, value=0}
          {arriveTangent=1.9017, interpMode=RCIM_Cubic, leaveTangent=1.9017, tangentMode=RCTM_User, time=1, value=0.5}
         ]}}}
      Scale Curve=980
     )
    ))
   &SolveForcesAndVelocity(Clamp Velocity=false, Limit Acceleration=false)
   &ScaleColor(
    Scale Alpha=&Multiply_Float(
     A=&FloatFromCurve(FloatCurve=DataInterface() {propertyValues {Curve {keys [
          {interpMode=RCIM_Cubic, tangentMode=RCTM_User, time=0, value=0}
          <at 0.2 0>
          {interpMode=RCIM_Cubic, tangentMode=RCTM_User, time=0.5, value=0.75}
          {tangentMode=RCTM_User, time=1, value=0}
         ]}}})
     B=1
    )
    Scale RGB=&Multiply_VectorByFloat(
     Vector=&MakeVectorFromLinearColorRGB(Color=&VariantColorSelector(
       Variant Amount=0
       Original Color=&ColorFromCurve(ColorCurve=DataInterface() {propertyValues {
          AlphaCurve {keys []}
          BlueCurve {keys [<at 0 0.5149 RCIM_Cubic>,<at 0.0079 0.0812>,<at 0.4788 0.0812>,<at 0.8661 0.0167>]}
          GreenCurve {keys [<at 0 1 RCIM_Cubic>,<at 0.0079 0.25>,<at 0.4788 0.25>,<at 0.8661 0.1>]}
          RedCurve {keys [<at 0 0.0297 RCIM_Cubic>,<at 0.0079 0.065>,<at 0.4788 0.065>,<at 0.8661 0>]}
         }})
       Variant 1 Color=&ColorFromCurve(ColorCurve=DataInterface() {propertyValues {
          AlphaCurve {keys []}
          BlueCurve {keys [<at 0.0079 0.25>,<at 0.4788 0.25>,<at 0.8661 0.1>]}
          GreenCurve {keys [<at 0.0079 0.1853>,<at 0.4788 0.1113>,<at 0.8661 0.0267>]}
          RedCurve {keys [<at 0.0079 0.065>,<at 0.4788 0.065>,<at 0.8661 0>]}
         }})
       Variant 2 Color=?
      ))
     Float=&Lerp_Float(A=0.5, B=5, Alpha=$Particles.HighSpeedParticle)
    )
    ScaleA=true
    ScaleRGB=true
   )
   &FN_Sprite_Attributes(
    Color=&MakeLinearColorFromVectorAndFloat(
     "Vector (RGB)"=&MakeVectorFromLinearColorRGB(Color=&VariantColorSelector(Variant Amount=0, Original Color=<rgba 0 1.5 0.75 1>, Variant 1 Color=<rgba 0 0.45 1.5 1>, Variant 2 Color=?))
     "Float (Alpha)"=&Multiply_Float(A=0.2, B=&FN_FrontEndFloatModifier(Gameplay=0.5, Front End=0.6))
    )
    Create Color Attribute=true
    Create SpriteSize Attribute=true
    SpriteSize=&Multiply_Vector2D(
     A=&Vector2DFromFloat(Value=&Multiply_Float(A=180, B=&Multiply_Float(A=$Particles.Random size, B=$Particles.Size Scale)))
     B=<0.6 1.6>
    )
    Write to Module Namespace=true
   ) {module=FN_Sprite_Attributes001}
  ]
  rendererClasses [&NiagaraSpriteRendererProperties, &NiagaraMeshRendererProperties]
  renderers [
   {rendererClass=&NiagaraSpriteRendererProperties}
   {rendererClass=&NiagaraSpriteRendererProperties, rendererIndex=1}
   {rendererClass=&NiagaraMeshRendererProperties, rendererIndex=2}
  ]
 }
 GPU_Sparks_Wheel {
  simTarget=GPUComputeSim
  emitterUpdateScript [
   &EmitterState(Inactive Response=::NewEnumerator1, Life Cycle Mode=::NewEnumerator1, Loop Delay=2.9, Loop Duration=1, UseLoopDelay=false)
   &SpawnRate(SpawnRate=&Lerp_Float(A=0, B=200, Alpha=$System.Timeline.Time), Use Spawn Probability=false) {module=SpawnRate001}
  ]
  particleSpawnScript [
   &PartitionParticles()
   /SetVariables(
    Particles.HighSpeedParticle:NiagaraFloat=&UniformAOrBFloat(A=0, Distribution Weight=0.01)
    Particles.Random size:NiagaraFloat=&RandomRangeFloat(Minimum=0.2, Maximum=0.3)
   ) {isSetParameters=true, module=SetVariables_#6}
   &InitializeParticle(
    Color=<rgba 10 10 10 1>
    Lifetime=&Lerp_Float(
     A=&RandomRangeFloat(Minimum=0.1, Maximum=0.5)
     B=&RandomRangeFloat(Minimum=0.2, Maximum=0.4)
     Alpha=$Particles.HighSpeedParticle
    )
    Sprite Rotation Mode=::NewEnumerator2
    Sprite Size=&Lerp_Vector2(A=&RandomRangeVector2D(Minimum=<1 15>, Maximum=<5 30>), B=<1 50>, Alpha=$Particles.HighSpeedParticle)
    Sprite Size Mode=::NewEnumerator4
    Sprite UV Mode=::NewEnumerator3
    UsePositionOffset=false
    Write Material Random=true
   ) {module=InitializeParticle001}
   &SkeletalMeshLocation(
    Bone / Socket Index=&Add_Integer(A=4, B=$Particles.PartitionParticles.Partition)
    Bone Sampling Mode=::NewEnumerator4
    Interpolate Mesh Sampling=false
    MaskSpawnGroup=false
    Sampled Position Offset=&AddVectorToPosition(Position=&RandomRangeVector(Minimum=<-5 -5 -5>, Maximum=<5 5 5>), Vector=<0 0 0>)
    Skeletal Mesh=$System.Skeletal Mesh
   )
   &ShapeLocation(
    Handle Radius=18
    Large Radius=40
    MaskSpawnGroup=false
    Override Seed=false
    Rotation Angle:NiagaraFloat=&Lerp_Float(A=-0.2, B=-0.3, Alpha=&MakeFloatFromInt(Int=$Particles.PartitionParticles.Partition))
    Rotation Mode=::NewEnumerator1
    Shape Primitive=::NewEnumerator3
    U Distribution:NiagaraFloat=&Lerp_Float(A=0.7, B=0.7, Alpha=&MakeFloatFromInt(Int=$Particles.PartitionParticles.Partition))
    UseOwnerScale=true
   )
   &CameraOffset(Camera Offset Amount=-50) {module=CameraOffset001}
   /SetVariables(Particles.DY3:Vector4f=&MakeVector4(X=3, Y=1, Z=1, W=1)) {isSetParameters=true, module=SetVariables_#8}
   /SetVariables(Particles.Velocity=&Multiply_Vector(
     A=$Particles.ShapeLocation.ShapeBitangent
     B=&VectorFromFloat(Value=&Lerp_Float(
       A=&RandomRangeFloat(Minimum=100, Maximum=200)
       B=&RandomRangeFloat(Minimum=1200, Maximum=800)
       Alpha=$Particles.HighSpeedParticle
      ))
    )) {isSetParameters=true, module=SetVariables_#10}
  ]
  particleUpdateScript [
   &ParticleState()
   /SetVariables(Particles.Size Scale:NiagaraFloat=&FloatFromCurve(
     FloatCurve=DataInterface() {propertyValues {Curve {keys [
         {interpMode=RCIM_Cubic, tangentMode=RCTM_User, time=0, value=0}
         {interpMode=RCIM_Cubic, tangentMode=RCTM_User, time=0.5, value=0.3}
         {tangentMode=RCTM_User, time=1, value=0}
        ]}}}
     Scale Curve=0.7
    )) {isSetParameters=true, module=SetVariables_#9}
   &CurlNoiseForce(Noise Frequency=50, Noise Strength=&Lerp_Float(A=100, B=100, Alpha=$Particles.HighSpeedParticle), Pan Noise=true)
   &ScaleSpriteSize(Scale Sprite Size Mode=::NewEnumerator0, Uniform Scale Factor=$Particles.Size Scale) {module=ScaleSpriteSize001}
   &SpriteRotationRate(Rotation Rate=&RandomRangeFloat(Minimum=-5, Maximum=5)) {module=SpriteRotationRate001}
   &Drag(Drag=&Lerp_Float(A=0.5, B=11, Alpha=$Particles.HighSpeedParticle), Use Linear Drag=true, Use Rotational Drag=true)
   &GravityForce(Gravity=&MakeVector(
     X=0
     Y=0
     Z=&FloatFromCurve(
      FloatCurve=DataInterface() {propertyValues {Curve {keys [
          {arriveTangent=0.215, interpMode=RCIM_Cubic, leaveTangent=0.215, tangentMode=RCTM_User, time=0.3, value=0}
          {arriveTangent=1.9017, interpMode=RCIM_Cubic, leaveTangent=1.9017, tangentMode=RCTM_User, time=1, value=0.5}
         ]}}}
      Scale Curve=980
     )
    ))
   &VortexForce(
    UseOriginOffset=false
    Vortex Axis=&RotateVectorByQuaternion(Vector To Rotate=<0 0 1>, Quaternion=$Particles.Initial.SkeletalMeshLocation.SampledRotation)
    Vortex Axis Coordinate Space:ENiagaraCoordinateSpace=::Local
    Vortex Force Amount=&Lerp_Float(A=0, B=400, Alpha=$Particles.HighSpeedParticle)
    Vortex Origin:NiagaraPosition=$Particles.Initial.SkeletalMeshLocation.SampledPosition
   ) {enabled=false}
   &SolveForcesAndVelocity(Clamp Velocity=false, Limit Acceleration=false)
   &ScaleColor(
    Scale Alpha=&Multiply_Float(
     A=&FloatFromCurve(FloatCurve=DataInterface() {propertyValues {Curve {keys [
          {interpMode=RCIM_Cubic, tangentMode=RCTM_User, time=0, value=0}
          <at 0.2 0>
          {interpMode=RCIM_Cubic, tangentMode=RCTM_User, time=0.5, value=0.75}
          {tangentMode=RCTM_User, time=1, value=0}
         ]}}})
     B=1
    )
    Scale RGB=&Multiply_VectorByFloat(
     Vector=&MakeVectorFromLinearColorRGB(Color=&VariantColorSelector(
       Variant Amount=0
       Original Color=&ColorFromCurve(ColorCurve=DataInterface() {propertyValues {
          AlphaCurve {keys []}
          BlueCurve {keys [<at 0 0.5149 RCIM_Cubic>,<at 0.0079 0.1452>,<at 0.4788 0.1452>,<at 0.8661 0.0167>]}
          GreenCurve {keys [<at 0 1 RCIM_Cubic>,<at 0.0079 0.25>,<at 0.4788 0.25>,<at 0.8661 0.1>]}
          RedCurve {keys [<at 0 0.0297 RCIM_Cubic>,<at 0.0079 0.065>,<at 0.4788 0.065>,<at 0.8661 0>]}
         }})
       Variant 1 Color=&ColorFromCurve(ColorCurve=DataInterface() {propertyValues {
          AlphaCurve {keys []}
          BlueCurve {keys [<at 0.0079 4>,<at 0.4788 0.25>,<at 0.8661 0.1>]}
          GreenCurve {keys [<at 0.0079 2.6189>,<at 0.4788 0.1113>,<at 0.8661 0.0317>]}
          RedCurve {keys [<at 0.0079 1.0405>,<at 0.4788 0.065>,<at 0.8661 0>]}
         }})
       Variant 2 Color=?
      ))
     Float=&Lerp_Float(A=1, B=4, Alpha=$Particles.HighSpeedParticle)
    )
    ScaleA=true
    ScaleRGB=true
   )
   &FN_Sprite_Attributes(
    Color=&MakeLinearColorFromVectorAndFloat(
     "Vector (RGB)"=&MakeVectorFromLinearColorRGB(Color=&MultiplyLinearColor(
       Linear Color=&VariantColorSelector(Variant Amount=0, Original Color=<rgba 0 1.5 0.675 1>, Variant 1 Color=<rgba 0 0.675 1.5 1>, Variant 2 Color=?)
       Scale Factor=&MakeVector4_FromFloat(Float=&Lerp_Float(A=1, B=2, Alpha=$Particles.HighSpeedParticle))
      ))
     "Float (Alpha)"=&Multiply_Float(A=0.2, B=&FN_FrontEndFloatModifier(Gameplay=0.5, Front End=0.6))
    )
    Create Color Attribute=true
    Create SpriteSize Attribute=true
    SpriteSize=&Multiply_Vector2D(
     A=&Vector2DFromFloat(Value=&Multiply_Float(A=180, B=&Multiply_Float(A=$Particles.Random size, B=$Particles.Size Scale)))
     B=<0.6 1.6>
    )
    Write to Module Namespace=true
   ) {module=FN_Sprite_Attributes001}
  ]
  rendererClasses [&NiagaraSpriteRendererProperties]
  renderers [
   {rendererClass=&NiagaraSpriteRendererProperties}
   {rendererClass=&NiagaraSpriteRendererProperties, rendererIndex=1}
  ]
 }
 Glow {
  emitterUpdateScript [
   &EmitterState(
    Inactive Response=::NewEnumerator1
    Life Cycle Mode=::NewEnumerator1
    Loop Behavior=::NewEnumerator1
    Loop Duration Mode=::NewEnumerator1
    UseLoopDelay=false
   )
   &SpawnBurst_Instantaneous(Spawn Count=4, Use Spawn Probability=false, UseLoopCountLimit=true)
   &SpawnBurst_Instantaneous(Spawn Count=4, Spawn Group=1, Use Spawn Probability=false, UseLoopCountLimit=false) {module=SpawnBurst_Instantaneous001}
   &SpawnBurst_Instantaneous(Spawn Count=4, Spawn Time=0.7021, Use Spawn Probability=false, UseLoopCountLimit=false) {module=SpawnBurst_Instantaneous002}
   &SpawnBurst_Instantaneous(Spawn Count=4, Spawn Time=0.8564, Use Spawn Probability=false, UseLoopCountLimit=false) {module=SpawnBurst_Instantaneous003}
  ]
  particleSpawnScript [
   &InitializeParticle(
    Lifetime=0
    Mesh Scale Mode=::NewEnumerator3
    Sprite Size=&Multiply_Vector2D_ByFloat(
     Float=&Lerp_Float(
      A=40
      B=20
      Alpha=&ModuloFloat(A=&MakeFloatFromInt(Int=&ReturnExecIndex()), B=2)
     )
     Vector2D=<1 4>
    )
    Sprite Size Mode=::NewEnumerator4
    UsePositionOffset=false
    Write Material Random=true
   )
   &PartitionParticles(Number Of Partitions=4)
   &PartitionParticles() {module=PartitionParticles001}
   /SetVariables(Particles.ConeMeshIndex:NiagaraInt32=$Particles.PartitionParticles.Partition) {isSetParameters=true}
   /SetVariables(Particles.VisibilityTag=$Emitter.SpawnGroup) {isSetParameters=true}
   &CameraOffset(Camera Offset Amount=3)
   /SetVariables(Particles.FoldOffset=4) {isSetParameters=true}
  ]
  particleUpdateScript [
   &ParticleState(Kill Particles When Lifetime Has Elapsed=false, Let Infinitely Lived Particles Die When Emitter Deactivates=true)
   &ScaleMeshSize(Scale Factor=$Engine.Owner.Scale)
   &SkeletalMeshLocation(
    Bone / Socket Index=$Particles.PartitionParticles.Partition
    Bone Sampling Mode=::NewEnumerator4
    MaskSpawnGroup=false
    Orientation Sampling=::NewEnumerator1
    Position Sampling=::NewEnumerator2
    Skeletal Mesh=$System.Skeletal Mesh
   )
   &SkeletalMeshLocation(
    Bone / Socket Index=$Particles.PartitionParticles.Partition
    Bone Sampling Mode=::NewEnumerator4
    MaskSpawnGroup=false
    Orientation Sampling=::NewEnumerator1
    Position Sampling=::NewEnumerator2
    Skeletal Mesh=$System.Skeletal Mesh
   ) {module=SkeletalMeshLocation003}
   &SkeletalMeshLocation(
    Bone / Socket Index=$Particles.PartitionParticles.Partition
    MaskSpawnGroup=false
    Mesh Sampling Type=::NewEnumerator5
    Orientation Sampling=::NewEnumerator1
    Skeletal Mesh=$System.Skeletal Mesh
    Socket Sampling Mode=::NewEnumerator5
    Write To Particle Position=false
   ) {module=SkeletalMeshLocation001}
   /SetVariables(
    Particles.AirGlow Color:LinearColor=&MultiplyLinearColor(
     Linear Color=&VariantColorSelector(
      Variant Amount=0
      Morning Color=<rgba 1 1 1 1>
      Day Color=<rgba 1 1 1 1>
      Evening Color=<rgba 0.4 0.4 0.4 2>
      Night Color=<rgba 1 1 1 1>
      Use TOD?=true
      Original Color=<rgba 0 1 0.3163 0.4>
      Variant 1 Color=<rgba 0 0.2333 0.4 0.7>
      Variant 2 Color=?
      Use FrontEnd?=true
      Front End Color=<rgba 1 1 1 0.05>
      Gameplay Color=<rgba 1 1 1 1>
     )
     Scale Factor=&MakeVector4(
      X=1
      Y=1
      Z=1
      W=&Lerp_Float(
       A=0
       B=1
       Alpha=&FloatFromCurve(
        FloatCurve=DataInterface() {propertyValues {Curve {keys [{tangentMode=RCTM_User, time=0, value=0}, {tangentMode=RCTM_User, time=1, value=1}]}}}
        CurveIndex=$System.Timeline.Time
       )
      )
     )
    )
    Particles.AirGlow Offset=10
    Particles.AirGlowSize:Vector2f=<30 30>
    Particles.Glow Color=&MultiplyLinearColor(
     Linear Color=&VariantColorSelector(
      Variant Amount=0
      Morning Color=<rgba 1 1 1 1>
      Day Color=<rgba 1 1 1 1>
      Evening Color=<rgba 0.3 0.3 0.3 1>
      Night Color=<rgba 1 1 1 0.3>
      Use TOD?=true
      Original Color=<rgba 0 20 7.6667 0.5>
      Variant 1 Color=<rgba 0 2.6667 20 1>
      Variant 2 Color=?
      Use FrontEnd?=true
      Front End Color=<rgba 1 1 1 1>
      Gameplay Color=<rgba 1 1 1 1>
     )
     Scale Factor=&MakeVector4(X=1, Y=1, Z=1, W=&Lerp_Float(A=0, B=1, Alpha=$System.Overshoot))
    )
    Particles.GlowRotation:Quat4f=&LerpQuaternion(
     Quaternion A:Quat4f=$Particles.SkeletalMeshLocation003.SampledRotation
     Quaternion B:Quat4f=/BRCosmetics/Athena/Effects/Particles/Skydiving/NS_Contrail_Apprentice/OrientMestToVectorFromCustomRot(
      InitialRotation:Quat4f=$Particles.SkeletalMeshLocation003.SampledRotation
      Look At Direction:Vector3f=<1 0 0>
      Reference Axis Vector:Vector3f=<0 1 0>
      RotationVector:Vector3f=<0 0.05 -0.23>
     )
     Lerp Factor:NiagaraFloat=&FN_FrontEndFloatModifier(Gameplay=0, Front End=1)
     Enable Spherical Linear Interpolation:NiagaraBool=false
    )
   ) {isSetParameters=true}
   /SetVariables(
    Particles.GlowPosition:NiagaraPosition=&AddVectorToPosition(
     Position=$Particles.SkeletalMeshLocation001.SampledPosition
     Vector=&TransformVector(
      Vector=&AddVectorToPosition(
       Position=<0 10 5>
       Vector=&SelectVectorFromArray(
        Direct Array Index=&ReturnExecIndex()
        Vector Selection Array=DataInterface() {propertyValues {FloatData [<0 0 0>, <0 -5 -4>, <1 0 0>, <0 -5 -5>]}}
       )
      )
      Destination Space=::Simulation
     )
    )
    Particles.Velocity=&RotateVectorByQuaternion(
     Vector To Rotate=&FN_FrontEndVector3Modifier(Gameplay=<0 1 0>, Front End=<0 -0.6 1>)
     Quaternion=$Particles.SkeletalMeshLocation001.SampledRotation
    )
   ) {isSetParameters=true}
   /SetVariables(
    Particles.Color=&MultiplyLinearColor(
     Linear Color=&VariantColorSelector(
      Variant Amount=0
      Morning Color=<rgba 1 1 1 1>
      Day Color=<rgba 1 1 1 1>
      Evening Color=<rgba 0.4 0.4 0.4 0.4>
      Night Color=<rgba 1 1 1 1>
      Use TOD?=true
      Original Color=<rgba 0.2056 4 2.035 1>
      Variant 1 Color=<rgba 0.3234 0.8478 4 1>
      Variant 2 Color=?
      Use FrontEnd?=true
      Front End Color=<rgba 1 1 1 0>
      Gameplay Color=<rgba 1 1 1 1>
     )
     Scale Factor=&MakeVector4(
      X=1
      Y=1
      Z=1
      W=&FloatFromCurve(
       FloatCurve=DataInterface() {propertyValues {Curve {keys [{tangentMode=RCTM_User, time=0, value=0}, {tangentMode=RCTM_User, time=1, value=1}]}}}
       CurveIndex=$System.Overshoot
      )
     )
    )
    Particles.MeshOrientation=$Particles.SkeletalMeshLocation.SampledRotation
    Particles.Position=&AddVectorToPosition(
     Position=$Particles.SkeletalMeshLocation.SampledPosition
     Vector=&TransformVector(Vector=<0 2 0>, Destination Space=::Simulation)
    )
   ) {isSetParameters=true}
   /SetVariables(
    Particles.GlowCardPosition:NiagaraPosition=&AddVectorToPosition(
     Position=$Particles.SkeletalMeshLocation001.SampledPosition
     Vector=&RotateVectorByQuaternion(
      Vector To Rotate=&Lerp_Vector(A=<0 0 0>, B=<0 0 0>, Alpha=&MakeFloatFromInt(Int=$Particles.Initial.PartitionParticles001.Partition))
      Quaternion=$Particles.SkeletalMeshLocation001.SampledRotation
     )
    )
    Particles.GlowCardScale:Vector3f=&Lerp_Vector(
     A=<1 1 1>
     B=<0.6 0.6 0.6>
     Alpha=&MakeFloatFromInt(Int=$Particles.Initial.PartitionParticles001.Partition)
    )
   ) {isSetParameters=true}
   /SetVariables(
    Particles.FoldColor:LinearColor=&MultiplyLinearColor(
     Linear Color=&VariantColorSelector(
      Variant Amount=0
      Morning Color=<rgba 1 1 1 1>
      Day Color=<rgba 1 1 1 1>
      Evening Color=<rgba 1 1 1 4>
      Night Color=<rgba 1 1 1 1>
      Use TOD?=true
      Original Color=<rgba 0.2056 4 2.035 1>
      Variant 1 Color=<rgba 0.0667 0 4 4>
      Variant 2 Color=?
      Use FrontEnd?=true
      Front End Color=<rgba 1 1 1 0>
      Gameplay Color=<rgba 1 1 1 1>
     )
     Scale Factor=&MakeVector4(
      X=1
      Y=1
      Z=1
      W=&FloatFromCurve(
       FloatCurve=DataInterface() {propertyValues {Curve {keys [{tangentMode=RCTM_User, time=1, value=0}, {tangentMode=RCTM_User, time=3.5, value=1}]}}}
       CurveIndex=$System.Timeline.Time
      )
     )
    )
    Particles.FoldPosition:NiagaraPosition=&SelectVectorFromArray(
     Direct Array Index=$Particles.Initial.PartitionParticles.Partition
     Vector Selection Array=DataInterface() {propertyValues {FloatData [<-24.5 62.2 4>, <-27 62 -1>, <25 60 3>, <27 62 -1>]}}
    )
   ) {isSetParameters=true}
   &DynamicMaterialParameters(
    Index 0 Param 2=&Multiply_Float(A=$Engine.Time, B=2)
    Index 0 Param 4=&Lerp_Float(
     A=0
     B=&SelectFloatFromArray(
      Direct Array Index=$Particles.PartitionParticles.Partition
      Float Selection Array=DataInterface() {propertyValues {FloatData [0.2,0.1,0.4,0]}}
     )
     Alpha=1
    )
    Module.Index0Param0Write=true
    Module.Index0Param1Write=true
    Module.Index0Param2Write=true
    Module.Index0Param3Write=true
   )
   &FN_Sprite_Attributes(
    Camera Offset=30
    Color=&MultiplyLinearColor(
     Linear Color=&VariantColorSelector(
      Variant Amount=0
      Morning Color=<rgba 1 1 1 1>
      Day Color=<rgba 1 1 1 1>
      Evening Color=<rgba 1 1 1 2>
      Night Color=<rgba 10 10 10 1>
      Use TOD?=true
      Original Color=<rgba 0 1 0.5 0.2>
      Variant 1 Color=<rgba 0.025 0 0.5 0.2>
      Variant 2 Color=?
     )
     Scale Factor=&MakeVector4(X=1, Y=1, Z=1, W=&Lerp_Float(A=0, B=1, Alpha=$System.Overshoot))
    )
    Create Camera Offset Attribute=true
    Create Color Attribute=true
    Create SpriteSize Attribute=true
    SpriteSize=&Vector2DFromFloat(Value=60)
    Write to Module Namespace=true
   )
   &FN_Light_Attributes(
    Color=&VariantColorSelector(Variant Amount=0, Original Color=<rgba 1.25 30 15.625 4>, Variant 1 Color=<rgba 0 1.5557 30 4>, Variant 2 Color=?)
    Create Color Attribute=true
    Create Exponent Attribute=true
    Create Radius Attribute=true
    Exponent=2
    Radius=20
    Scale Color=&Lerp_Float(A=0, B=0.2, Alpha=$System.Overshoot)
   )
   /SetVariables(Particles.FoldScale:Vector3f=&VectorFromFloat(Value=&FloatFromCurve(FloatCurve=DataInterface() {propertyValues {Curve {keys [<at 0 0>,<at 1 1>]}}}, CurveIndex=$System.Overshoot))) {enabled=false, isSetParameters=true}
  ]
  rendererClasses [
   &NiagaraMeshRendererProperties
   &NiagaraSpriteRendererProperties
   &NiagaraLightRendererProperties
  ]
  renderers [
   {rendererClass=&NiagaraMeshRendererProperties, rendererIndex=0}
   {rendererClass=&NiagaraMeshRendererProperties, rendererIndex=1}
   {rendererClass=&NiagaraMeshRendererProperties, rendererIndex=2}
   {rendererClass=&NiagaraSpriteRendererProperties, rendererIndex=3}
   {rendererClass=&NiagaraSpriteRendererProperties, rendererIndex=4}
   {rendererClass=&NiagaraLightRendererProperties, rendererIndex=5}
   {rendererClass=&NiagaraSpriteRendererProperties, rendererIndex=6}
   {rendererClass=&NiagaraMeshRendererProperties, rendererIndex=7}
  ]
 }
 SmokeLargeLocalSpace {
  bEnabled=false
  emitterUpdateScript [
   &EmitterState(Inactive Response=::NewEnumerator1, Life Cycle Mode=::NewEnumerator1, Loop Delay=0.15, Loop Duration=9.9017, UseLoopDelay=false)
   &SpawnBurst_Instantaneous(Spawn Count=2, Use Spawn Probability=true, UseLoopCountLimit=false) {enabled=false}
   &SpawnPerFrame(
    Spawn Probability:NiagaraFloat=&FloatFromCurve(FloatCurve=DataInterface() {propertyValues {Curve {keys [<at -1 1>,<at 0 0>]}}}, CurveIndex=$User.ForwardDot)
    Use Spawn Probability=true
   ) {enabled=false}
   &SpawnRate(SpawnRate=33, Use Spawn Probability=true)
  ]
  particleSpawnScript [
   &FN_SetSpawnGroup(Number of Spawn Groups=2)
   &PartitionParticles(Number Of Partitions=4)
   &InitializeParticle(
    Color=&MakeLinearColorFromVectorAndFloat("Vector (RGB)"=<1 1 1>, "Float (Alpha)"=1)
    Lifetime=0.55
    Position Offset=&SelectVectorFromArray(
     Direct Array Index=&SpawnGroup()
     Vector Selection Array=DataInterface() {propertyValues {FloatData [<85 60 25>, <85 -60 25>]}}
    )
    Sprite Size Mode=::NewEnumerator3
    Uniform Sprite Size=77
    UsePositionOffset=true
    Write Material Random=true
   )
  ]
  particleUpdateScript [
   &ParticleState()
   &SkeletalMeshLocation(
    Bone / Socket Index=$Particles.PartitionParticles.Partition
    MaskSpawnGroup=false
    Mesh Sampling Type=::NewEnumerator5
    Position Sampling=::NewEnumerator1
    Sampled Position Offset=&AddVectorToPosition(Position=&RandomRangeVector(Minimum=<-5 -5 -5>, Maximum=<5 5 5>), Vector=<0 -20 0>)
    Skeletal Mesh=$System.Skeletal Mesh
    Socket Sampling Mode=::NewEnumerator5
    Source Movement Scale:Vector3f=&VectorFromFloat(Value=&Lerp_Float(
      A=&FloatFromCurve(FloatCurve=DataInterface() {propertyValues {Curve {keys [<at 0.7 0>,<at 1 0>]}}})
      B=&FloatFromCurve(FloatCurve=DataInterface() {propertyValues {Curve {keys [<at 0 1>,<at 0.7 1>,<at 1 0>]}}})
      Alpha=$System.Overshoot
     ))
   )
   &Color(
    Color=&ColorFromCurve(ColorCurve=DataInterface() {propertyValues {
       AlphaCurve {keys [<at 0 0.2>,<at 0.1 1>,<at 0.75 0.666>,<at 1 0>]}
       BlueCurve {keys [<at 0 0.6>, {arriveTangent=0.0577, time=0.4743, value=0.075}, <at 1 0.05>]}
       GreenCurve {keys [<at 0 1>, {arriveTangent=-0.0961, time=0.4743, value=0.075}, <at 1 0.05>]}
       RedCurve {keys [<at 0 0>, {arriveTangent=-1.0957, time=0.4743, value=0.075}, <at 1 0.05>]}
      }})
    Scale Alpha=&MakeFloatFromLinearColor(LinearColor=$Particles.Initial.Color, Channel:ENiagaraLinearColor_Channels=::NewEnumerator3)
   )
   &ScaleSpriteSize(
    Uniform Curve Scale=&RandomRangeFloat(Minimum=0.8, Maximum=1.1)
    Uniform Curve Sprite Scale=DataInterface() {propertyValues {Curve {keys [<at 0 0>, {leaveTangent=1.5237, time=0.1, value=0.6}, <at 0.6 1.1>, {arriveTangent=-5.0995, time=1, value=0.5}]}}}
   )
   &DynamicMaterialParameters(
    Index 0 Param 1=&Add_Float(
     A=&Multiply_Float(A=$Engine.Time, B=1)
     B=&RandomRangeFloat(Minimum=0, Maximum=1)
    )
    Index 0 Param 3=&FloatFromCurve(
     FloatCurve=DataInterface() {propertyValues {Curve {keys [<at 0 1>,<at 0.333 0>]}}}
     CurveIndex=$Particles.Age
     Scale Curve=&Multiply_Float(A=0.02, B=0.1)
    )
    Index 2 Param 1=1
    Index 2 Param 2=&RandomRangeFloat(Minimum=0, Maximum=1)
    Index 2 Param 3=&FloatFromCurve(FloatCurve=DataInterface() {propertyValues {Curve {keys [
         {interpMode=RCIM_Cubic, tangentMode=RCTM_Break, time=0.666, value=0}
         {arriveTangent=10.9325, interpMode=RCIM_Cubic, tangentMode=RCTM_Break, time=1, value=2}
        ]}}})
    Index 2 Param 4=1
    Module.Param0WriteEnabled=true
    Module.Param0WriteEnabled1=true
    Module.Param0WriteEnabled2=true
    Module.Param1WriteEnabled=true
    Module.Param1WriteEnabled1=true
    Module.Param1WriteEnabled2=true
    Module.Param2WriteEnabled=true
    Module.Param2WriteEnabled1=true
    Module.Param2WriteEnabled2=true
    Module.Param3WriteEnabled=true
    Module.Param3WriteEnabled1=true
    Module.Param3WriteEnabled2=true
    Write Parameter Index 1=true
    Write Parameter Index 2=true
   )
   /SetVariables(Particles.Position=&AddVectorToPosition(
     Position=$Particles.Position
     Vector=&Multiply_VectorByFloat(
      Vector=&NormalizeVector(
       Vector To Normalize=&Lerp_Vector(
        A=<0 1 0>
        B=&TransformVector(
         Vector=&NormalizeVector(
          Vector To Normalize=&Multiply_Vector(A=$Engine.Owner.Velocity, B=<0 -1 -3>)
          Safe Normalize=true
          Safe Fallback Threshold=0
          Safe Fallback Vector=<0 0 0.001>
         )
         Source Space=::World
         Destination Space=::Local
        )
        Alpha=&FloatFromCurve(
         FloatCurve=DataInterface() {propertyValues {Curve {keys [{interpMode=RCIM_Cubic, leaveTangent=2.5306, tangentMode=RCTM_Break, time=0, value=0}, {tangentMode=RCTM_User, time=1, value=1}]}}}
         Scale Curve=$System.ForwardAndRightDot
        )
       )
       Safe Normalize=true
       Safe Fallback Threshold=0
       Safe Fallback Vector=<0 0 0.001>
      )
      Float=14
     )
    )) {isSetParameters=true}
  ]
  rendererClasses [&NiagaraSpriteRendererProperties]
  renderers [
   {rendererClass=&NiagaraSpriteRendererProperties}
   {rendererClass=&NiagaraSpriteRendererProperties, rendererIndex=1}
  ]
 }
 SmokeSmallLocalSpace {
  emitterUpdateScript [
   &EmitterState(Life Cycle Mode=::NewEnumerator1, Loop Delay=0.15, Loop Duration=0, UseLoopDelay=false)
   &SpawnRate(Spawn Probability=0.4, SpawnRate=400, Use Spawn Probability=true)
  ]
  particleSpawnScript [
   &PartitionParticles(Number Of Partitions=4)
   &InitializeParticle(
    Color=<rgba 0 0 0 0.1>
    Lifetime=&RandomRangeFloat(Minimum=0.2, Maximum=0.4)
    Position Offset Coordinate Space=::Local
    Sprite Size Mode=::NewEnumerator3
    Uniform Sprite Size=&Multiply_Float(A=20, B=&MakeFloatFromVector(Vector=$Engine.Owner.Scale, Channel=::NewEnumerator0))
    UsePositionOffset=true
    Write Material Random=true
   )
   &SkeletalMeshLocation(
    Bone / Socket Index=$Particles.PartitionParticles.Partition
    Interpolate Mesh Sampling=false
    MaskSpawnGroup=false
    Mesh Sampling Type=::NewEnumerator5
    Skeletal Mesh=$System.Skeletal Mesh
    Socket Sampling Mode=::NewEnumerator5
   ) {module=SkeletalMeshLocation001}
   &CameraOffset()
   &AddVelocity(Velocity=<0 200 0>)
  ]
  particleUpdateScript [
   &ParticleState()
   &ScaleSpriteSize(
    Uniform Curve Scale=&RandomRangeFloat(Minimum=0.8, Maximum=1)
    Uniform Curve Sprite Scale=DataInterface() {propertyValues {Curve {keys [<at 0 1>, {arriveTangent=1.5721, time=1, value=2}]}}}
   )
   &DynamicMaterialParameters(
    Index 0 Param 1=&Add_Float(
     A=&Multiply_Float(A=$Engine.Time, B=1)
     B=&RandomRangeFloat(Minimum=0, Maximum=1)
    )
    Index 1 Param 1=9
    Index 1 Param 4=0.5
    Module.Param0WriteEnabled=true
    Module.Param0WriteEnabled1=true
    Module.Param0WriteEnabled2=true
    Module.Param1WriteEnabled=true
    Module.Param1WriteEnabled1=true
    Module.Param1WriteEnabled2=true
    Module.Param2WriteEnabled=true
    Module.Param2WriteEnabled1=true
    Module.Param2WriteEnabled2=true
    Module.Param3WriteEnabled=true
    Module.Param3WriteEnabled1=true
    Module.Param3WriteEnabled2=true
    Write Parameter Index 1=true
    Write Parameter Index 2=true
   )
   &GravityForce(Gravity=<0 0 400>)
   &Drag(Drag=&RandomRangeFloat(Minimum=2, Maximum=4), Use Linear Drag=true, Use Rotational Drag=true)
   &SolveForcesAndVelocity(Clamp Velocity=false, Limit Acceleration=false)
   &ScaleColor(
    Scale Alpha=&Multiply_Float(
     A=&FloatFromCurve(FloatCurve=DataInterface() {propertyValues {Curve {keys [<at 0 1>,<at 1 0>]}}})
     B=&Multiply_Float(A=$System.Timeline.InverseNormalizedTime, B=$System.Timeline001.InverseNormalizedTime)
    )
    ScaleA=true
    ScaleRGB=true
   )
  ]
  rendererClasses [&NiagaraSpriteRendererProperties, &NiagaraMeshRendererProperties]
  renderers [
   {rendererClass=&NiagaraSpriteRendererProperties}
   {rendererClass=&NiagaraMeshRendererProperties, rendererIndex=1}
  ]
 }
 SmokeSmallLocalSpace001 {
  emitterUpdateScript [
   &EmitterState(Life Cycle Mode=::NewEnumerator1, Loop Delay=0.15, Loop Duration=0, UseLoopDelay=false)
   &SpawnRate(Spawn Probability=0.7, SpawnRate=20, Use Spawn Probability=true)
  ]
  particleSpawnScript [
   &PartitionParticles(Number Of Partitions=4)
   &InitializeParticle(
    Color=<rgba 0 0 0 0.4>
    Lifetime=&RandomRangeFloat(Minimum=0.2, Maximum=0.4)
    Position Offset Coordinate Space=::Local
    Sprite Size Mode=::NewEnumerator3
    Uniform Sprite Size=&Multiply_Float(A=30, B=&MakeFloatFromVector(Vector=$Engine.Owner.Scale, Channel=::NewEnumerator0))
    UsePositionOffset=true
    Write Material Random=true
   )
   &SkeletalMeshLocation(
    Bone / Socket Index=$Particles.PartitionParticles.Partition
    Interpolate Mesh Sampling=false
    MaskSpawnGroup=false
    Mesh Sampling Type=::NewEnumerator5
    Skeletal Mesh=$System.Skeletal Mesh
    Socket Sampling Mode=::NewEnumerator5
   ) {module=SkeletalMeshLocation001}
   &CameraOffset()
   &AddVelocity(Velocity=<0 400 0>)
  ]
  particleUpdateScript [
   &ParticleState()
   &ScaleSpriteSize(Uniform Curve Sprite Scale=DataInterface() {propertyValues {Curve {keys [
        {interpMode=RCIM_Cubic, tangentMode=RCTM_User, time=0, value=0.5}
        {arriveTangent=1.5721, interpMode=RCIM_Cubic, tangentMode=RCTM_User, time=1, value=2}
       ]}}})
   &DynamicMaterialParameters(
    Index 0 Param 1=&Add_Float(
     A=&Multiply_Float(A=$Engine.Time, B=1)
     B=&RandomRangeFloat(Minimum=0, Maximum=1)
    )
    Index 1 Param 1=9
    Index 1 Param 4=0.5
    Module.Param0WriteEnabled=true
    Module.Param0WriteEnabled1=true
    Module.Param0WriteEnabled2=true
    Module.Param1WriteEnabled=true
    Module.Param1WriteEnabled1=true
    Module.Param1WriteEnabled2=true
    Module.Param2WriteEnabled=true
    Module.Param2WriteEnabled1=true
    Module.Param2WriteEnabled2=true
    Module.Param3WriteEnabled=true
    Module.Param3WriteEnabled1=true
    Module.Param3WriteEnabled2=true
    Write Parameter Index 1=true
    Write Parameter Index 2=true
   )
   &GravityForce(Gravity=<0 0 900>)
   &Drag(Drag=&RandomRangeFloat(Minimum=2, Maximum=5), Use Linear Drag=true, Use Rotational Drag=true)
   &SolveForcesAndVelocity(Clamp Velocity=false, Limit Acceleration=false)
   &ScaleColor(
    Scale Alpha=&Multiply_Float(
     A=&FloatFromCurve(FloatCurve=DataInterface() {propertyValues {Curve {keys [<at 0 1>,<at 1 0>]}}})
     B=&Multiply_Float(A=$System.Timeline.InverseNormalizedTime, B=$System.Timeline001.InverseNormalizedTime)
    )
    ScaleA=true
    ScaleRGB=true
   )
  ]
  rendererClasses [&NiagaraSpriteRendererProperties, &NiagaraMeshRendererProperties]
  renderers [
   {rendererClass=&NiagaraSpriteRendererProperties}
   {rendererClass=&NiagaraMeshRendererProperties, rendererIndex=1}
  ]
 }
 Sparks_Cutoff_Burst {
  bEnabled=false
  emitterUpdateScript [
   &EmitterState(Life Cycle Mode=::NewEnumerator1, Loop Behavior=::NewEnumerator1, Loop Delay=2.9, Loop Duration Mode=::NewEnumerator1, UseLoopDelay=false)
   /SetVariables(Emitter.Age:NiagaraFloat=&Lerp_Float(
     A=0
     B=$Emitter.Age
     Alpha=&FloatFromCurve(
      FloatCurve=DataInterface() {propertyValues {Curve {keys [<at 0 0 RCIM_Constant>,<at 0.2 1 RCIM_Constant>,<at 0.3 0 RCIM_Constant>]}}}
      CurveIndex=$System.Timeline.Float
     )
    )) {enabled=false, isSetParameters=true}
   &SpawnPerFrame(
    Spawn:NiagaraBool=&SetBoolByFloatComparison(A=$User.ForwardDot, B=0, Comparison Type=::NewEnumerator6)
    Spawn Count:NiagaraInt32=&RandomRangeInt(Minimum=2, Maximum=5)
    Use Spawn Probability=false
   )
   &SpawnBurst_Instantaneous(Age:NiagaraFloat=$Emitter.Age, Spawn Count=40, Spawn Time=0.1, Use Spawn Probability=false, UseLoopCountLimit=false) {enabled=false}
   &SpawnRate(SpawnRate=&Lerp_Float(A=0, B=50, Alpha=$System.Overshoot), Use Spawn Probability=false) {enabled=false, module=SpawnRate001}
  ]
  particleSpawnScript [
   /SetVariables(
    Particles.HighSpeedParticle:NiagaraFloat=&UniformAOrBFloat(A=0, Distribution Weight=0.2)
    Particles.Random size:NiagaraFloat=&RandomRangeFloat(Minimum=0.2, Maximum=0.3)
   ) {isSetParameters=true, module=SetVariables_#6}
   /SetVariables() {isSetParameters=true, module=SetVariables_#7}
   &PartitionParticles(Number Of Partitions=4)
   &InitializeParticle(
    Color=<rgba 30 30 30 1>
    Lifetime=&RandomRangeFloat(Minimum=0.3, Maximum=0.6)
    Sprite Rotation Mode=::NewEnumerator2
    Sprite Size Max=<3 10>
    Sprite Size Min=<2 10>
    Sprite Size Mode=::NewEnumerator2
    Sprite UV Mode=::NewEnumerator3
    UsePositionOffset=false
    Write Material Random=true
   ) {module=InitializeParticle001}
   &DynamicMaterialParameters(
    Index 0 Param 2=&RandomRangeFloat(Minimum=0, Maximum=4)
    Index 0 Param 4=2.5
    Module.Param0WriteEnabled=true
    Module.Param1WriteEnabled=true
    Module.Param2WriteEnabled=false
    Module.Param3WriteEnabled=true
   ) {module=DynamicMaterialParameters001}
   &AddVelocity(
    Cone Angle=&Lerp_Float(A=20, B=25, Alpha=$Particles.HighSpeedParticle)
    Cone Axis=<0 1 0.8>
    Module.UseConeFalloff=false
    Module.UseVelDistribution=false
    Override Seed=false
    Velocity Mode=::NewEnumerator2
    Velocity Speed=&Multiply_Float(
     A=1.5
     B=&Lerp_Float(
      A=&RandomRangeFloat(Minimum=150, Maximum=300)
      B=&RandomRangeFloat(Minimum=350, Maximum=400)
      Alpha=$Particles.HighSpeedParticle
     )
    )
   )
   &CameraOffset(Camera Offset Amount=-50) {module=CameraOffset001}
   /SetVariables(Particles.DY3:Vector4f=&MakeVector4(X=3, Y=1, Z=1, W=1)) {isSetParameters=true, module=SetVariables_#8}
  ]
  particleUpdateScript [
   &ParticleState()
   /SetVariables(Particles.Size Scale:NiagaraFloat=&FloatFromCurve(
     FloatCurve=DataInterface() {propertyValues {Curve {keys [
         {interpMode=RCIM_Cubic, tangentMode=RCTM_User, time=0, value=0}
         {interpMode=RCIM_Cubic, tangentMode=RCTM_User, time=0.5, value=0.3}
         {tangentMode=RCTM_User, time=1, value=0}
        ]}}}
     Scale Curve=0.7
    )) {isSetParameters=true, module=SetVariables_#9}
   &SkeletalMeshLocation(
    Bone / Socket Index=$Particles.PartitionParticles.Partition
    MaskSpawnGroup=false
    Mesh Sampling Type=::NewEnumerator5
    Position Sampling=::NewEnumerator1
    Sampled Position Offset=&AddVectorToPosition(Position=&RandomRangeVector(Minimum=<-5 -5 -5>, Maximum=<5 5 5>), Vector=<0 -20 0>)
    Skeletal Mesh=$System.Skeletal Mesh
    Socket Sampling Mode=::NewEnumerator5
    Source Movement Scale:Vector3f=&VectorFromFloat(Value=&Lerp_Float(
      A=&FloatFromCurve(FloatCurve=DataInterface() {propertyValues {Curve {keys [<at 0.7 0>,<at 1 0>]}}})
      B=&FloatFromCurve(FloatCurve=DataInterface() {propertyValues {Curve {keys [<at 0 1>,<at 0.7 1>,<at 1 0>]}}})
      Alpha=$System.Overshoot
     ))
   )
   &CurlNoiseForce(Noise Frequency=50, Noise Strength=&Lerp_Float(A=100, B=400, Alpha=$Particles.HighSpeedParticle), Pan Noise=true)
   &ScaleSpriteSize(Scale Sprite Size Mode=::NewEnumerator0, Uniform Scale Factor=$Particles.Size Scale) {module=ScaleSpriteSize001}
   &SpriteRotationRate(Rotation Rate=&RandomRangeFloat(Minimum=-5, Maximum=5)) {module=SpriteRotationRate001}
   &Drag(Drag=&RandomRangeFloat(Minimum=0.5, Maximum=5), Use Linear Drag=true, Use Rotational Drag=true)
   &GravityForce(Gravity=&MakeVector(
     X=0
     Y=0
     Z=&FloatFromCurve(
      FloatCurve=DataInterface() {propertyValues {Curve {keys [
          {arriveTangent=0.215, interpMode=RCIM_Cubic, leaveTangent=0.215, tangentMode=RCTM_User, time=0.3, value=0}
          {arriveTangent=1.9017, interpMode=RCIM_Cubic, leaveTangent=1.9017, tangentMode=RCTM_User, time=1, value=0.5}
         ]}}}
      Scale Curve=980
     )
    ))
   &SolveForcesAndVelocity(Clamp Velocity=false, Limit Acceleration=false)
   &ScaleColor(
    Scale Alpha=&Multiply_Float(
     A=&FloatFromCurve(FloatCurve=DataInterface() {propertyValues {Curve {keys [
          {interpMode=RCIM_Cubic, tangentMode=RCTM_User, time=0, value=0}
          {interpMode=RCIM_Cubic, tangentMode=RCTM_User, time=0.1, value=0.75}
          {tangentMode=RCTM_User, time=1, value=0}
         ]}}})
     B=1
    )
    Scale RGB=&Multiply_VectorByFloat(
     Vector=&MakeVectorFromLinearColorRGB(Color=&ColorFromCurve(ColorCurve=DataInterface() {propertyValues {
         AlphaCurve {keys []}
         BlueCurve {keys [<at 0 0.5149 RCIM_Cubic>,<at 0.0079 0.0812>,<at 0.4788 0.0812>,<at 0.8661 0.0167>]}
         GreenCurve {keys [<at 0 1 RCIM_Cubic>,<at 0.0079 0.25>,<at 0.4788 0.25>,<at 0.8661 0.1>]}
         RedCurve {keys [<at 0 0.0297 RCIM_Cubic>,<at 0.0079 0.065>,<at 0.4788 0.065>,<at 0.8661 0>]}
        }}))
     Float=&Lerp_Float(A=0.5, B=5, Alpha=$Particles.HighSpeedParticle)
    )
    ScaleA=true
    ScaleRGB=true
   )
   &FN_Sprite_Attributes(
    Color=&MakeLinearColorFromVectorAndFloat(
     "Vector (RGB)"=&MakeVectorFromLinearColorRGB(Color=&MultiplyLinearColor(Linear Color=<rgba 0 1.5 0.75 1>, Scale Factor=&MakeVector4_FromFloat(Float=1)))
     "Float (Alpha)"=&Multiply_Float(A=0.2, B=&FN_FrontEndFloatModifier(Gameplay=0.5, Front End=0.6))
    )
    Create Color Attribute=true
    Create SpriteSize Attribute=true
    SpriteSize=&Multiply_Vector2D(
     A=&Vector2DFromFloat(Value=&Multiply_Float(A=180, B=&Multiply_Float(A=$Particles.Random size, B=$Particles.Size Scale)))
     B=<0.6 1.6>
    )
    Write to Module Namespace=true
   ) {module=FN_Sprite_Attributes001}
  ]
  rendererClasses [&NiagaraSpriteRendererProperties, &NiagaraMeshRendererProperties]
  renderers [
   {rendererClass=&NiagaraSpriteRendererProperties}
   {rendererClass=&NiagaraSpriteRendererProperties, rendererIndex=1}
   {rendererClass=&NiagaraMeshRendererProperties, rendererIndex=2}
   {rendererClass=&NiagaraSpriteRendererProperties, rendererIndex=3}
  ]
 }
 Splines_Flame {
  emitterSpawnScript [/SetVariables(Emitter.UseSplineThicken=true) {isSetParameters=true}]
  emitterUpdateScript [
   &EmitterState(
    Inactive Response=::NewEnumerator1
    Life Cycle Mode=::NewEnumerator1
    Loop Behavior=::NewEnumerator1
    Loop Duration Mode=::NewEnumerator1
    UseLoopDelay=false
   )
   &SpawnBurst_Instantaneous(Spawn Count=4, Use Spawn Probability=false, UseLoopCountLimit=true)
   /SetVariables(Emitter.FlameAppearCurve:NiagaraFloat=&FloatFromCurve(FloatCurve=DataInterface() {propertyValues {Curve {keys [<at 0 0>,<at 1 1>]}}}, CurveIndex=$System.Timeline.Time)) {isSetParameters=true}
  ]
  particleSpawnScript [
   &PartitionParticles(Number Of Partitions=4)
   &InitializeParticle(
    Color=&VariantColorSelector(Variant Amount=0, Original Color=<rgba 0.6 0.6 0.6 1.5>, Variant 1 Color=<rgba 0.3067 0.2586 1.7 1.5>, Variant 2 Color=?)
    Mesh Scale=<3 1 3>
    Mesh Scale Mode=::NewEnumerator4
    UsePositionOffset=false
    Write Material Random=true
   )
   &CameraOffset()
   /SetVariables(Particles.SplineSize:Vector4f=&MakeVector4(
     X=&Lerp_Float(
      A=30
      B=20
      Alpha=&MakeFloatFromInt(Int=&ModuloInt(A=$Particles.PartitionParticles.Partition, B=2))
     )
     Y=&Lerp_Float(
      A=120
      B=80
      Alpha=&MakeFloatFromInt(Int=&ModuloInt(A=$Particles.PartitionParticles.Partition, B=2))
     )
     Z=0
     W=0
    )) {isSetParameters=true}
  ]
  particleUpdateScript [
   &ParticleState(Kill Particles When Lifetime Has Elapsed=false, Let Infinitely Lived Particles Die When Emitter Deactivates=true)
   &SkeletalMeshLocation(
    Bone / Socket Index=$Particles.PartitionParticles.Partition
    MaskSpawnGroup=false
    Mesh Sampling Type=::NewEnumerator5
    Skeletal Mesh=$System.Skeletal Mesh
    Socket Sampling Mode=::NewEnumerator5
   )
   &ScaleMeshSize(Scale Factor=&SelectVectorFromArray(
     Direct Array Index=&ModuloInt(A=$Particles.PartitionParticles.Partition, B=2)
     Vector Selection Array=DataInterface() {propertyValues {FloatData [<1 1 0.4>, <1 0.4 0.33>]}}
    ))
   &ScaleMeshSize(Scale Factor=&MakeVector(
     X=1
     Y=&FloatFromCurve(
      FloatCurve=DataInterface() {propertyValues {Curve {keys [{interpMode=RCIM_Cubic, tangentMode=RCTM_Break, time=0, value=0}, {interpMode=RCIM_Cubic, tangentMode=RCTM_Break, time=0.1, value=1}]}}}
      CurveIndex=$System.Timeline.Time
      Scale Curve=2
     )
     Z=1
    )) {module=ScaleMeshSize001}
   &ScaleMeshSize(Scale Factor=&MakeVector(X=1, Y=1, Z=1)) {module=ScaleMeshSize002}
   &DynamicMaterialParameters(
    Index 0 Param 1=&FloatFromCurve(
     FloatCurve=DataInterface() {propertyValues {Curve {keys [
         {arriveTangent=-37.8208, interpMode=RCIM_Cubic, leaveTangent=-37.8208, tangentMode=RCTM_User, time=0, value=10}
         {arriveTangent=-7.379, interpMode=RCIM_Cubic, leaveTangent=-7.379, tangentMode=RCTM_User, time=1, value=1}
        ]}}}
     CurveIndex=$System.Timeline.Time
    )
    Index 0 Param 2=&Multiply_Float(A=$Engine.Time, B=3)
    Index 0 Param 4=&Lerp_Float(
     A=0
     B=&SelectFloatFromArray(
      Direct Array Index=$Particles.PartitionParticles.Partition
      Float Selection Array=DataInterface() {propertyValues {FloatData [0.2,0.1,0.4,0]}}
     )
     Alpha=1
    )
    Module.Index0Param0Write=true
    Module.Index0Param1Write=true
    Module.Index0Param2Write=true
    Module.Index0Param3Write=true
   )
   &ScaleColor(
    Scale Alpha=&Lerp_Float(A=0, B=1, Alpha=$System.Overshoot)
    Scale RGB=&MakeVectorFromLinearColorRGB(Color=<rgba 1 1 1 1>)
    ScaleA=true
    ScaleRGB=true
   )
   /SetVariables(Particles.Color1:LinearColor=&MakeLinearColorFromVectorAndFloat(
     "Vector (RGB)"=&MakeVectorFromLinearColorRGB(Color=&MultiplyLinearColor(
       Linear Color=&VariantColorSelector(Variant Amount=0, Original Color=<rgba 1.2 1.2 1.2 1>, Variant 1 Color=<rgba 1.0119 0.3741 2.5 1>, Variant 2 Color=?)
       Scale Factor=&MakeVector4_FromFloat(Float=&FN_FrontEndFloatModifier(Gameplay=0.9, Front End=0.9))
      ))
     "Float (Alpha)"=&Lerp_Float(A=0, B=1, Alpha=$Emitter.FlameAppearCurve)
    )) {isSetParameters=true}
   /SetVariables(Particles.Color2=&MakeLinearColorFromVectorAndFloat(
     "Vector (RGB)"=&MakeVectorFromLinearColorRGB(Color=&MultiplyLinearColor(
       Linear Color=&VariantColorSelector(Variant Amount=0, Original Color=<rgba 1 1 1 1>, Variant 1 Color=<rgba 0 0.8351 3 1>, Variant 2 Color=?)
       Scale Factor=&MakeVector4_FromFloat(Float=0.7)
      ))
     "Float (Alpha)"=&Lerp_Float(A=0, B=0.35, Alpha=$Emitter.FlameAppearCurve)
    )) {isSetParameters=true}
   /SetVariables(Particles.Color3:LinearColor=&MakeLinearColorFromVectorAndFloat(
     "Vector (RGB)"=&MakeVectorFromLinearColorRGB(Color=<rgba 0.5 0.5 0.5 1>)
     "Float (Alpha)"=&FN_FrontEndFloatModifier(Gameplay=&Lerp_Float(A=0, B=0.2, Alpha=$Emitter.FlameAppearCurve), Front End=0.1)
    )) {isSetParameters=true}
   /SetVariables(Particles.Color4:LinearColor=&MultiplyLinearColor(
     Linear Color=&VariantColorSelector(Variant Amount=0, Original Color=<rgba 0 0.5 0.2333 2>, Variant 1 Color=<rgba 0 0.3747 2 2>, Variant 2 Color=?)
     Scale Factor=&MakeVector4(X=1, Y=1, Z=1, W=$System.Overshoot)
    )) {isSetParameters=true}
   &self:ScaleColorsByTOD(
    Linear Color:LinearColor=$Particles.Color
    Linear Color001:LinearColor=$Particles.Color1
    Linear Color002:LinearColor=$Particles.Color2
    Linear Color003:LinearColor=$Particles.Color4
    TOD:NiagaraFloat=$System.TOD
   ) {module=ScaleColorsByTOD}
   &InterpolateOverTime(
    Rate Of Change:NiagaraFloat=$System.Interpolate Rate
    Target Vector Value:NiagaraPosition=$Particles.Position
    Variable Type=::NewEnumerator2
   )
   /SetVariables(Particles.Interpolate DY:Vector4f=&MakeFloat4FromVectorAndFloat(
     "Vector (RGB)"=$Particles.InterpolateOverTime.InterpolateOverTime_V3_.Moving Average
     "Float (Alpha)"=&Lerp_Float(
      A=3
      B=1.5
      Alpha=&ModuloFloat(A=&MakeFloatFromInt(Int=$Particles.PartitionParticles.Partition), B=2)
     )
    )) {isSetParameters=true}
  ]
  rendererClasses [&NiagaraMeshRendererProperties]
  renderers [
   {rendererClass=&NiagaraMeshRendererProperties, rendererIndex=0}
   {rendererClass=&NiagaraMeshRendererProperties, rendererIndex=1}
   {rendererClass=&NiagaraMeshRendererProperties, rendererIndex=2}
   {rendererClass=&NiagaraMeshRendererProperties, rendererIndex=3}
  ]
 }
 WheelFX {
  emitterSpawnScript [
   /SetVariables(Emitter.TileX=1, Emitter.TileY=4) {isSetParameters=true}
   /SetVariables(Emitter.AngleFade=2) {isSetParameters=true}
  ]
  emitterUpdateScript [
   &EmitterState(
    Inactive Response=::NewEnumerator1
    Life Cycle Mode=::NewEnumerator1
    Loop Behavior=::NewEnumerator1
    Loop Duration Mode=::NewEnumerator1
    UseLoopDelay=false
   )
   &SpawnBurst_Instantaneous(Spawn Count=2, Use Spawn Probability=false, UseLoopCountLimit=false)
   &SpawnBurst_Instantaneous(Spawn Count=4, Spawn Group=1, Use Spawn Probability=false, UseLoopCountLimit=false) {module=SpawnBurst_Instantaneous001}
  ]
  particleSpawnScript [
   &PartitionParticles()
   &InitializeParticle(
    Color=&VariantColorSelector(Variant Amount=0, Original Color=<rgba 1 1 1 1>, Variant 1 Color=<rgba 1.1561 0.5861 4 1>, Variant 2 Color=?)
    Lifetime=2
    Mesh Scale Mode=::NewEnumerator3
    Mesh Uniform Scale=&Multiply_Float(
     A=&SelectFloatFromArray(Direct Array Index=0, Float Selection Array=DataInterface() {propertyValues {FloatData [1,-1]}})
     B=1
    )
    UsePositionOffset=false
    Write Material Random=true
   )
   &InitialMeshOrientation(
    Mesh Axis To Orient=<0 1 0>
    Mesh Orientation Mode=::NewEnumerator2
    Orientation Coordinate Space=::Local
    Rotation=&SelectVectorFromArray(
     Direct Array Index=&ReturnExecIndex()
     Vector Selection Array=DataInterface() {propertyValues {FloatData [<0 0 0>, <0 -0.25 0>, <0 -0.3 0>, <0 0 0>]}}
    )
    Use Rotation Vector=true
   )
   /SetVariables(Particles.MeshIndex=$Particles.PartitionParticles.Partition) {isSetParameters=true}
   /SetVariables(Particles.VisibilityTag=$Emitter.SpawnGroup) {isSetParameters=true}
  ]
  particleUpdateScript [
   &ParticleState(Kill Particles When Lifetime Has Elapsed=false, Let Infinitely Lived Particles Die When Emitter Deactivates=true)
   &ScaleMeshSize(Scale Factor=$Engine.Owner.Scale)
   &DynamicMaterialParameters(
    Index 0 Param 1=&Lerp_Float(A=10, B=0, Alpha=$System.Timeline.Time)
    Index 0 Param 2=&Multiply_Float(A=$Engine.Time, B=&Add_Float(A=0, B=-8))
    Module.Index0Param0Write=true
    Module.Index0Param1Write=true
    Module.Index0Param2Write=true
    Module.Index0Param3Write=true
   )
   &SkeletalMeshLocation(
    Bone / Socket Index=&Add_Integer(A=$Particles.PartitionParticles.Partition, B=4)
    Bone Sampling Mode=::NewEnumerator4
    MaskSpawnGroup=false
    Orientation Sampling=::NewEnumerator1
    Sampled Position Offset=<0 0 -0.5>
    Skeletal Mesh=$System.Skeletal Mesh
   )
   &SolveForcesAndVelocity(Clamp Velocity=false, Limit Acceleration=false)
   &FN_Sprite_Attributes(
    Camera Offset=40
    Color=&MultiplyLinearColor(
     Linear Color=&VariantColorSelector(Variant Amount=0, Original Color=<rgba 0 1 0.5 0.2>, Variant 1 Color=<rgba 0 0.3667 1 0.4>, Variant 2 Color=?)
     Scale Factor=&MakeVector4(X=1, Y=1, Z=1, W=&Lerp_Float(A=0, B=0.5, Alpha=$System.Overshoot))
    )
    Create Camera Offset Attribute=true
    Create Color Attribute=true
    Create Position Attribute=true
    Create SpriteSize Attribute=true
    Position:NiagaraPosition=&AddVectorToPosition(
     Position=$Particles.SkeletalMeshLocation.SampledPosition
     Vector=&Lerp_Vector(A=<0 -30 25>, B=<0 30 25>, Alpha=&MakeFloatFromInt(Int=$Particles.PartitionParticles.Partition))
    )
    SpriteSize=&Vector2DFromFloat(Value=120)
    Write to Module Namespace=true
   )
   &FN_Sprite_Attributes(
    Alignment:Vector3f=<1 0 0>
    Color=&MultiplyLinearColor(
     Linear Color=<rgba 0.2 0.1397 0.0549 1>
     Scale Factor=&MakeVector4(X=1, Y=1, Z=1, W=$System.Timeline.Float)
    )
    Create Alignment Attribute=true
    Create Camera Offset Attribute=true
    Create Color Attribute=true
    Create Facing Attribute=true
    Create Position Attribute=true
    Create Sprite Rotation Attribute=true
    Create SpriteSize Attribute=true
    Facing:Vector3f=<1 0 0>
    Position:NiagaraPosition=&AddVectorToPosition(
     Position=$Particles.SkeletalMeshLocation.SampledPosition
     Vector=&SelectVectorFromArray(
      Direct Array Index=&ReturnParticleIDAttributes()
      Vector Selection Array=DataInterface() {propertyValues {FloatData [<10 0 0>, <10 0 0>, <-10 0 0>, <-10 0 0>, <10 0 0>, <10 0 0>]}}
     )
    )
    SpriteRotation:NiagaraFloat=&Multiply_Float(A=$Engine.Time, B=4)
    SpriteSize=&Multiply_Vector2D_ByFloat(
     Float=50
     Vector2D=&Vector2DFromFloat(Value=&MakeFloatFromVector(Vector=$Engine.Owner.Scale, Channel=::NewEnumerator0))
    )
    Write to Module Namespace=true
   ) {module=FN_Sprite_Attributes001}
   &FN_Mesh_Attributes(
    Color=&MultiplyLinearColor(
     Linear Color=&VariantColorSelector(Variant Amount=0, Original Color=<rgba 0.0919 3 1.6324 1>, Variant 1 Color=<rgba 0.0919 0.8674 3 1>, Variant 2 Color=?)
     Scale Factor=&MakeVector4(X=1, Y=1, Z=1, W=&Lerp_Float(A=0.2, B=1, Alpha=$System.Overshoot))
    )
    Create Color Attribute=true
    Write to Module Namespace=true
   )
   /SetVariables() {isSetParameters=true}
   &ScaleColor(
    Scale Alpha=&Lerp_Float(A=0, B=1, Alpha=$System.Overshoot)
    Scale RGB=&Multiply_Vector(A=<1 1 1>, B=&VectorFromFloat(Value=$System.TOD))
    ScaleA=true
    ScaleRGB=true
   )
  ]
  rendererClasses [&NiagaraMeshRendererProperties, &NiagaraSpriteRendererProperties]
  renderers [
   {rendererClass=&NiagaraMeshRendererProperties, rendererIndex=0}
   {rendererClass=&NiagaraMeshRendererProperties, rendererIndex=1}
   {rendererClass=&NiagaraMeshRendererProperties, rendererIndex=2}
   {rendererClass=&NiagaraMeshRendererProperties, rendererIndex=3}
   {rendererClass=&NiagaraMeshRendererProperties, rendererIndex=4}
   {rendererClass=&NiagaraSpriteRendererProperties, rendererIndex=5}
   {rendererClass=&NiagaraMeshRendererProperties, rendererIndex=6}
   {rendererClass=&NiagaraSpriteRendererProperties, rendererIndex=7}
  ]
 }
}
summary {
 systemName=NS_Glider_GateHound_Trail
 emitters [
  {emitterName=Splines_Flame, rendererClasses [&NiagaraMeshRendererProperties]}
  {
   emitterName=Glow
   rendererClasses [
    &NiagaraMeshRendererProperties
    &NiagaraSpriteRendererProperties
    &NiagaraLightRendererProperties
   ]
  }
  {
   emitterName=GPU_Sparks
   simTarget=GPUComputeSim
   rendererClasses [&NiagaraSpriteRendererProperties, &NiagaraMeshRendererProperties]
  }
  {emitterName=WheelFX, rendererClasses [&NiagaraMeshRendererProperties, &NiagaraSpriteRendererProperties]}
  {emitterName=GPU_Sparks_Wheel, simTarget=GPUComputeSim, rendererClasses [&NiagaraSpriteRendererProperties]}
  {emitterName=FN_Debug, rendererClasses [&NiagaraSpriteRendererProperties]}
  {emitterName=CPU_Sparks, rendererClasses [&NiagaraSpriteRendererProperties, &NiagaraMeshRendererProperties]}
  {
   emitterName=CPU_Sparks_Wheel
   rendererClasses [&NiagaraSpriteRendererProperties, &NiagaraMeshRendererProperties]
  }
  {
   bEnabled=false
   emitterName=Sparks_Cutoff_Burst
   rendererClasses [&NiagaraSpriteRendererProperties, &NiagaraMeshRendererProperties]
  }
  {bEnabled=false, emitterName=SmokeLargeLocalSpace, rendererClasses [&NiagaraSpriteRendererProperties]}
  {
   emitterName=SmokeSmallLocalSpace
   rendererClasses [&NiagaraSpriteRendererProperties, &NiagaraMeshRendererProperties]
  }
  {
   emitterName=SmokeSmallLocalSpace001
   rendererClasses [&NiagaraSpriteRendererProperties, &NiagaraMeshRendererProperties]
  }
 ]
 userVariables [
  User.bIsFrontEnd {type=NiagaraBool}
  User.bIsFrontEndPreview {type=NiagaraBool}
  User.bIsFullyDeployed {type=NiagaraBool}
  User.DEBUG {type=NiagaraFloat, defaultValue=0, description=""}
  User.DoorOffsets {type=NiagaraPosition, description="", defaultValue {}}
  User.ForwardDot {type=NiagaraFloat, defaultValue=0, description=""}
  User.ParaGlideLeanAlpha {type=NiagaraFloat, defaultValue=0, description=""}
  User.RightDot {type=NiagaraFloat, defaultValue=0, description=""}
  User.ThrottlePositions {
   type=NiagaraDataInterfaceArrayFloat3
   description=""
   defaultValue {
    struct=NiagaraExt_VariableValue_DataInterface
    value {
     dataInterface=&self:NiagaraDataInterfaceArrayFloat3_0
     dataInterfaceClass=&NiagaraDataInterfaceArrayFloat3
    }
   }
  }
  User.Variant {type=NiagaraFloat, defaultValue=1, description=Choose Variant}
 ]
}
