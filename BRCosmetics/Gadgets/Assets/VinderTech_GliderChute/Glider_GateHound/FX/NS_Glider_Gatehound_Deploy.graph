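# Niagara system for the deploy effect of the Gatehound glider: mesh doors, portal smoke, flares with lights, bike smoke, flames and a ribbon trail.
# Most emitters follow bones or sockets on a shared skeletal mesh, staged by loop delays and User.Beat2.
# MainTrail drives four ribbon renderers through separate color and width attributes.
# Colors switch by cosmetic variant and dim by time of day; several values change for front-end preview.
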NiagaraSystem /BRCosmetics/Gadgets/Assets/VinderTech_GliderChute/Glider_GateHound/FX/NS_Glider_Gatehound_Deploy
dependencies {
 usedDataInterfaces [
  &NiagaraDataInterfaceCurve
  &NiagaraDataInterfaceColorCurve
  &NiagaraDataInterfaceVector2DCurve
  &NiagaraDataInterfaceSpriteRendererInfo
  &NiagaraDataInterfaceArrayFloat
  &NiagaraDataInterfaceSkeletalMesh
 ]
 usedDynamicInputs [&FNTimeOfDayScalar]
 usedModules [
  &self:SystemSpawnScript.SystemScriptSource.SystemScriptGraph.NiagaraNodeAssignment_0.SetVariables_#1
  &FN_VariantManager
  &SystemState
  &self:SystemSpawnScript.SystemScriptSource.SystemScriptGraph.NiagaraNodeAssignment_1.SetVariables_#2
 ]
}
emitters {
 Bike_Flare {
  emitterUpdateScript [
   &EmitterState(
    Inactive Response=::NewEnumerator1
    Life Cycle Mode=::NewEnumerator1
    Loop Behavior=::NewEnumerator1
    Loop Delay=0.145
    Loop Duration=1
    UseLoopDelay=true
   )
   &SpawnBurst_Instantaneous(Spawn Count=1, Use Spawn Probability=false)
  ]
  particleSpawnScript [
   &InitializeParticle(
    Color=&VariantColorSelector(
     Variant Amount=0
     Morning Color=<rgba 1 1 1 0.05>
     Day Color=<rgba 1 1 1 0.04>
     Evening Color=<rgba 1 1 1 0.04>
     Night Color=<rgba 1 1 1 0.002>
     Use TOD?=true
     Original Color=<rgba 0 7 5.3667 1>
     Variant 1 Color=<rgba 0 0.2333 7 1>
     Variant 2 Color=?
    )
    Sprite Size=<124.2746 3>
    Sprite Size Mode=::NewEnumerator4
    UsePositionOffset=true
    Write Material Random=true
   )
   &SkeletalMeshLocation(
    Bone / Socket Index=&Add_Integer(A=&ReturnExecIndex(), B=1)
    MaskSpawnGroup=false
    Mesh Sampling Type=::NewEnumerator5
    Skeletal Mesh=$System.Skeletal Mesh
    Socket Sampling Mode=::NewEnumerator5
   ) {module=SkeletalMeshLocation001}
  ]
  particleUpdateScript [
   &SolveForcesAndVelocity(Clamp Velocity=false, Limit Acceleration=false)
   &ParticleState()
   &SkeletalMeshLocation(
    Bone / Socket Index=&Add_Integer(A=&ReturnExecIndex(), B=1)
    MaskSpawnGroup=false
    Mesh Sampling Type=::NewEnumerator5
    Skeletal Mesh=$System.Skeletal Mesh
    Socket Sampling Mode=::NewEnumerator5
   )
   &ScaleSpriteSize(
    Uniform Curve Index=$Emitter.Age
    Uniform Curve Sprite Scale=DataInterface() {propertyValues {Curve {keys [{tangentMode=RCTM_User, time=0.5, value=0}, {tangentMode=RCTM_User, time=1, value=5}, <at 1.2 0>]}}}
   ) {module=ScaleSpriteSize002}
   /SetVariables(Particles.DistanceTraveled:NiagaraFloat=&DistanceBetweenPositions(A=$Particles.Initial.Position, B=$Particles.Position)) {isSetParameters=true, module=SetVariables_#3}
   &CameraOffset(Camera Offset Amount=&FloatFromCurve(FloatCurve=DataInterface() {propertyValues {Curve {keys [<at 0.8 1>,<at 1 0>]}}}, CurveIndex=$Particles.Age, Scale Curve=40))
   &FN_Sprite_Attributes(
    Color=&VariantColorSelector(
     Variant Amount=0
     Morning Color=<rgba 1 1 1 1>
     Day Color=<rgba 1 1 1 0.4>
     Evening Color=<rgba 1 1 1 0.4>
     Night Color=<rgba 1 1 1 0.15>
     Use TOD?=true
     Original Color=<rgba 0 1.5 0.7 0.1>
     Variant 1 Color=<rgba 0 0.025 1.5 0.1>
     Variant 2 Color=?
    )
    Create Color Attribute=true
    Create SpriteSize Attribute=true
    SpriteSize=&Multiply_Vector2D_ByFloat(
     Float=&FloatFromCurve(
      FloatCurve=DataInterface() {propertyValues {Curve {keys [<at 0.4 0>, {tangentMode=RCTM_User, time=0.5, value=2}, {tangentMode=RCTM_User, time=1, value=5}, <at 1.2 0>]}}}
      CurveIndex=$Emitter.Age
     )
     Vector2D=<100 100>
    )
    Write to Module Namespace=true
   )
   &FN_Sprite_Attributes(
    Color=&MultiplyLinearColor(
     Linear Color=&VariantColorSelector(Variant Amount=0, Original Color=<rgba 0 0.04 0.0073 1>, Variant 1 Color=<rgba 0 0.0313 0.04 1>, Variant 2 Color=?)
     Scale Factor=&MakeVector4(
      X=1
      Y=1
      Z=1
      W=&FloatFromCurve(FloatCurve=DataInterface() {propertyValues {Curve {keys [
           <at 0.2 0 RCIM_Cubic>
           {arriveTangent=2, interpMode=RCIM_Cubic, leaveTangent=2, time=0.3, value=1}
           {interpMode=RCIM_Cubic, tangentMode=RCTM_User, time=0.7, value=1}
           {interpMode=RCIM_Cubic, tangentMode=RCTM_User, time=1, value=0}
          ]}}})
     )
    )
    Create Color Attribute=true
    Create SpriteSize Attribute=true
    SpriteSize=&Vector2DFromFloat(Value=250)
    Write to Module Namespace=true
   ) {module=FN_Sprite_Attributes001}
  ]
  rendererClasses [&NiagaraSpriteRendererProperties]
  renderers [
   {rendererClass=&NiagaraSpriteRendererProperties}
   {rendererClass=&NiagaraSpriteRendererProperties, rendererIndex=1}
   {rendererClass=&NiagaraSpriteRendererProperties, rendererIndex=2}
  ]
 }
 Bike_Smoke {
  emitterUpdateScript [
   &EmitterState()
   &SpawnRate(
    SpawnRate=&FloatFromCurve(
     FloatCurve=DataInterface() {propertyValues {Curve {keys [
         {interpMode=RCIM_Cubic, tangentMode=RCTM_User, time=-0.01, value=0}
         {interpMode=RCIM_Cubic, tangentMode=RCTM_User, time=0, value=1}
         {arriveTangent=-6.0176, interpMode=RCIM_Cubic, leaveTangent=-6.0176, tangentMode=RCTM_User, time=0.4, value=0}
        ]}}}
     CurveIndex=&Subtract_Float(A=$Emitter.Age, B=$User.Beat2)
     Scale Curve=30
    )
    Use Spawn Probability=false
   )
  ]
  particleSpawnScript [
   &InitializeParticle(
    Color Mode=::NewEnumerator0
    Lifetime Max=0.7
    Lifetime Min=0.7
    Lifetime Mode=::NewEnumerator1
    Sprite Rotation Mode=::NewEnumerator2
    Sprite Size Mode=::NewEnumerator1
    Uniform Sprite Size Max=120
    Uniform Sprite Size Min=80
    UsePositionOffset=false
    Write Material Random=true
   )
   &SkeletalMeshLocation(
    "Inherit Velocity (Scale)"=8
    Inherited Velocity Speed Limit=80
    MaskSpawnGroup=false
    Mesh Sampling Type=::NewEnumerator5
    Skeletal Mesh=$System.Skeletal Mesh
    Socket Sampling Mode=::NewEnumerator5
    Velocity Sampling=::NewEnumerator0
    VelocitySpeedLimit=true
   )
   &AddVelocity(Module.Use Falloff Distance=false, Override Seed=false, Velocity Mode=::NewEnumerator1, Velocity Speed=100)
   &AddVelocity(Velocity=<0 0 0>) {module=AddVelocity001}
  ]
  particleUpdateScript [
   &ParticleState()
   &SkeletalMeshLocation(
    Bone Sampling Mode=::NewEnumerator6
    MaskSpawnGroup=false
    Orientation Sampling=::NewEnumerator1
    Position Sampling=::NewEnumerator2
    Skeletal Mesh=$System.Skeletal Mesh
   ) {module=SkeletalMeshLocation001}
   &SpriteRotationRate(Rotation Rate=&RandomRangeFloat(Minimum=-150, Maximum=150))
   &ScaleSpriteSize(
    Uniform Curve Scale=&MakeCustomFloatFromBool(Bool=$User.bIsFrontEndPreview, True Float=0.7, False Float=0.9)
    Uniform Curve Sprite Scale=DataInterface() {propertyValues {Curve {keys [
        {arriveTangent=1.6237, interpMode=RCIM_Cubic, leaveTangent=1.6237, tangentMode=RCTM_User, time=0, value=1}
        {arriveTangent=0.637, tangentMode=RCTM_User, time=1, value=1.7}
       ]}}}
   )
   &Drag(Drag=3, Use Linear Drag=true, Use Rotational Drag=true)
   &SolveForcesAndVelocity(Clamp Velocity=false, Limit Acceleration=false)
   &ScaleColor(
    Scale Alpha=&FloatFromCurve(
     FloatCurve=DataInterface() {propertyValues {Curve {keys [{interpMode=RCIM_Cubic, tangentMode=RCTM_User, time=0, value=1}, {interpMode=RCIM_Cubic, tangentMode=RCTM_User, time=1, value=0}]}}}
     Scale Curve=&MakeCustomFloatFromBool(Bool=$User.bIsFrontEndPreview, True Float=0.5, False Float=1)
    )
    ScaleA=true
    ScaleRGB=true
   )
   &Color(Color=&MultiplyLinearColor(
     Linear Color=&LerpLinearColors(
      StartColor=<rgba 0.01 0.01 0.01 1.7>
      EndColor=&MultiplyLinearColor(
       Linear Color=&VariantColorSelector(Variant Amount=0, Original Color=<rgba 0 8 2.6667 1>, Variant 1 Color=<rgba 0.08 0.74 8 1>, Variant 2 Color=?)
       Scale Factor=&MakeVector4_FromLinearColor(LinearColor=&FNTimeOfDayColor(Night Color=<rgba 0.25 0.25 0.25 1>, Morning Color=<rgba 1 1 1 1>, Day Color=<rgba 1 1 1 1>, Evening Color=<rgba 0.2 0.2 0.2 1>))
      )
      LerpFactor=&FloatFromCurve(
       FloatCurve=DataInterface() {propertyValues {Curve {keys [<at 500 0 RCIM_Cubic>, {arriveTangent=0.0052, interpMode=RCIM_Cubic, leaveTangent=0.0052, tangentMode=RCTM_User, time=980, value=1}]}}}
       CurveIndex=&DistanceBetweenPositions(A=$Particles.SkeletalMeshLocation001.SampledPosition, B=$Particles.Position)
      )
     )
     Scale Factor=&MakeVector4_FromLinearColor(LinearColor=&ColorFromCurve(ColorCurve=DataInterface() {propertyValues {
         AlphaCurve {keys [<at 0 1 RCIM_Cubic>,<at 1 0 RCIM_Cubic>]}
         BlueCurve {keys [<at 0.2203 1 RCIM_Cubic>,<at 0.6431 0 RCIM_Cubic>]}
         GreenCurve {keys [<at 0.2203 1 RCIM_Cubic>,<at 0.6431 0 RCIM_Cubic>]}
         RedCurve {keys [<at 0.2203 1 RCIM_Cubic>,<at 0.6431 0 RCIM_Cubic>]}
        }}))
    ))
  ]
  rendererClasses [&NiagaraSpriteRendererProperties]
  renderers [{rendererClass=&NiagaraSpriteRendererProperties}]
 }
 Doors {
  emitterUpdateScript [
   &EmitterState(Inactive Response=::NewEnumerator1, Life Cycle Mode=::NewEnumerator1, Loop Behavior=::NewEnumerator1, Loop Duration=4, UseLoopDelay=false)
   &SpawnBurst_Instantaneous(Spawn Count=2, Use Spawn Probability=false, UseLoopCountLimit=true)
   &SpawnBurst_Instantaneous(Spawn Group=1, Use Spawn Probability=false, UseLoopCountLimit=true) {module=SpawnBurst_Instantaneous001}
  ]
  particleSpawnScript [
   /SetVariables(Particles.spawngroup:NiagaraInt32=$Emitter.SpawnGroup) {isSetParameters=true}
   &PartitionParticles()
   &InitializeParticle(
    Color=&VariantColorSelector(Variant Amount=0, Original Color=<rgba 0.012 0.02 0.02 1.2>, Variant 1 Color=<rgba 0.06 0.0747 0.1 1.2>, Variant 2 Color=?)
    Lifetime=&SelectFloatFromArray(Direct Array Index=$Particles.spawngroup, Float Selection Array=DataInterface() {propertyValues {FloatData [4,2]}})
    Mass Max=2
    Mass Mode=::NewEnumerator3
    Mesh Scale Mode=::NewEnumerator3
    Sprite Size Mode=::NewEnumerator3
    Uniform Sprite Size=400
    UsePositionOffset=false
    Write Material Random=true
   )
   /SetVariables(Particles.MeshIndex=$Particles.PartitionParticles.Partition) {isSetParameters=true}
  ]
  particleUpdateScript [
   &ParticleState()
   &SkeletalMeshLocation(
    Bone / Socket Index=&MergeHelper_IntSwitch(
     0=&Add_Integer(A=1, B=$Particles.PartitionParticles.Partition)
     1=0
     2=0
     3=0
     4=0
     5=?
     Index=$Particles.spawngroup
     Num=::NewEnumerator3
    )
    Bone Sampling Mode=::NewEnumerator4
    MaskSpawnGroup=false
    Sampled Position Offset=&MergeHelper_VectorSwitch(0=<0 0 0>, 1=<0 0 50>, 2=<0 0 0>, 3=<0 0 0>, 4=<0 0 0>, Index=$Particles.spawngroup, Num=::NewEnumerator3)
    Skeletal Mesh=$System.Skeletal Mesh
   )
   &ScaleMeshSize(Scale Factor=$Particles.SkeletalMeshLocation.SampledScale)
   &ScaleSpriteSize(Uniform Curve Sprite Scale=DataInterface() {propertyValues {Curve {keys [
        {interpMode=RCIM_Cubic, tangentMode=RCTM_User, time=0, value=0}
        {interpMode=RCIM_Cubic, tangentMode=RCTM_User, time=0.15, value=1}
        {arriveTangent=-1.1765, interpMode=RCIM_Cubic, leaveTangent=-1.1765, time=0.8, value=1}
        {tangentMode=RCTM_User, time=1, value=0}
       ]}}})
   &DynamicMaterialParameters(
    Index 0 Param 1=&FloatFromCurve(FloatCurve=DataInterface() {propertyValues {Curve {keys [<at 0 1 RCIM_Cubic>,<at 0.8 0 RCIM_Cubic>]}}})
    Index 0 Param 2=2
    Index 0 Param 3=1
    Index 0 Param 4=1
    Index 1 Param 1=&FloatFromCurve(FloatCurve=DataInterface() {propertyValues {Curve {keys [
         {arriveTangent=-2.3163, interpMode=RCIM_Cubic, time=0, value=2}
         {arriveTangent=-1.9019, interpMode=RCIM_Cubic, leaveTangent=-1.9019, tangentMode=RCTM_User, time=1, value=0.4}
        ]}}})
    Index 1 Param 2=0.4
    Index 1 Param 3=0.6
    Index 1 Param 4=1
    Index 2 Param 1=&FloatFromCurve(FloatCurve=DataInterface() {propertyValues {Curve {keys [<at 0.46 0>,<at 1 1>]}}})
    Index 2 Param 2=45
    Module.Param0WriteEnabled=true
    Module.Param0WriteEnabled1=false
    Module.Param0WriteEnabled2=true
    Module.Param1WriteEnabled=true
    Module.Param1WriteEnabled1=true
    Module.Param1WriteEnabled2=true
    Module.Param2WriteEnabled=true
    Module.Param2WriteEnabled1=true
    Module.Param2WriteEnabled2=true
    Module.Param3WriteEnabled=true
    Module.Param3WriteEnabled1=true
    Module.Param3WriteEnabled2=true
    Write Parameter Index 1=true
    Write Parameter Index 2=true
   )
   &ScaleColor(
    Scale Alpha=&FloatFromCurve(
     FloatCurve=DataInterface() {propertyValues {Curve {keys [
         {interpMode=RCIM_Cubic, tangentMode=RCTM_Break, time=0, value=0}
         <at 0.2 1 RCIM_Constant>
         {interpMode=RCIM_Cubic, tangentMode=RCTM_Break, time=0.5, value=1}
         {arriveTangent=-0.1838, leaveTangent=-8.0089, time=0.6, value=0}
        ]}}}
     Scale Curve=&FN_FrontEndFloatModifier(Gameplay=1, Front End=0)
    )
    ScaleA=true
    ScaleRGB=true
   ) {module=ScaleColor001}
  ]
  rendererClasses [&NiagaraMeshRendererProperties, &NiagaraSpriteRendererProperties]
  renderers [
   {rendererClass=&NiagaraMeshRendererProperties, rendererIndex=0}
   {rendererClass=&NiagaraMeshRendererProperties, rendererIndex=1}
   {rendererClass=&NiagaraSpriteRendererProperties, rendererIndex=2}
  ]
 }
 Flames {
  emitterUpdateScript [
   &EmitterState(Life Cycle Mode=::NewEnumerator1, Loop Behavior=::NewEnumerator1, Loop Delay=1.1, Loop Duration=0.35, UseLoopDelay=true)
   &SpawnRate(SpawnRate=50, Use Spawn Probability=false)
  ]
  particleSpawnScript [
   &InitializeParticle(
    Color=<rgba 0.5 0.5 0.5 1>
    Lifetime Max=0.6
    Lifetime Min=0.2
    Lifetime Mode=::NewEnumerator1
    Sprite Rotation Mode=::NewEnumerator2
    Sprite Size Mode=::NewEnumerator1
    Uniform Sprite Size Max=80
    Uniform Sprite Size Min=60
    UsePositionOffset=false
    Write Material Random=true
   ) {module=InitializeParticle001}
   &DynamicMaterialParameters(
    Index 0 Param 2=&UniformRangedFloat(Minimum=0, Maximum=1)
    Index 0 Param 4=&UniformRangedFloat(Minimum=3, Maximum=5)
    Module.Param0WriteEnabled=false
    Module.Param1WriteEnabled=true
    Module.Param2WriteEnabled=false
    Module.Param3WriteEnabled=true
   )
   &SkeletalMeshLocation(
    MaskSpawnGroup=false
    Mesh Sampling Type=::NewEnumerator5
    Sampled Position Offset=<0 -50 -10>
    Sampled Position Offset Coordinate Space:ENiagaraSkelMeshTransforms=::NewEnumerator2
    Skeletal Mesh=$System.Skeletal Mesh
    Socket Sampling Mode=::NewEnumerator5
    Velocity Sampling=::NewEnumerator0
    VelocitySpeedLimit=true
   )
   &ShapeLocation(
    Handle Radius=0
    Large Radius=&FN_FrontEndFloatModifier(Gameplay=2, Front End=5)
    MaskSpawnGroup=false
    Override Seed=false
    Rotation Mode=::NewEnumerator1
    Shape Primitive=::NewEnumerator3
    UseOwnerScale=true
   )
   &AddVelocityFromPoint(
    Module.Use Falloff Distance=false
    Velocity Origin:Vector3f=&SimulationPosition()
    Velocity Strength=&Multiply_Float(A=100, B=&FN_FrontEndFloatModifier(Gameplay=1, Front End=0.4))
   )
   &CameraOffset(Camera Offset Amount=&FN_FrontEndFloatModifier(Gameplay=30, Front End=10))
  ]
  particleUpdateScript [
   &ParticleState()
   &ScaleSpriteSize(Scale Factor=&Vector2DFromFloat(Value=&FloatFromCurve(
      FloatCurve=DataInterface() {propertyValues {Curve {keys [
          {arriveTangent=2.4033, interpMode=RCIM_Cubic, leaveTangent=2.4033, tangentMode=RCTM_User, time=0.0002, value=0.15}
          <at 0.4996 0.9697>
          {arriveTangent=-0.8166, interpMode=RCIM_Cubic, leaveTangent=-0.8166, tangentMode=RCTM_User, time=1.009, value=0.713}
         ]}}}
      Scale Curve=&FN_FrontEndFloatModifier(Gameplay=0.7, Front End=0.4)
     )))
   &SubUVAnimation(
    Random Start Frame=true
    "Sprite Renderer "=DataInterface() {propertyValues {SpriteRenderer=&self:Flames_0.NiagaraSpriteRendererProperties_0}}
    UseEndFrame=true
    UseStartFrame=true
   )
   &DynamicMaterialParameters(
    Index 0 Param 1=&FloatFromCurve(FloatCurve=DataInterface() {propertyValues {Curve {keys [<at 0 0>,<at 0.7 1>,<at 1.0011 1>]}}})
    Index 0 Param 2=2.5
    Index 0 Param 3=1
    Index 0 Param 4=2
    Module.Param0WriteEnabled=true
    Module.Param1WriteEnabled=false
    Module.Param2WriteEnabled=true
    Module.Param3WriteEnabled=true
   ) {module=DynamicMaterialParameters001}
   &GravityForce(Gravity=&MakeVector(
     X=0
     Y=0
     Z=&FloatFromCurve(FloatCurve=DataInterface() {propertyValues {Curve {keys [<at 0 0>,<at 0.6 0>,<at 0.7124 0.9779>,<at 1 1>]}}}, Scale Curve=200)
    ))
   &Drag(Drag=2.5, Use Linear Drag=true, Use Rotational Drag=false)
   &SolveForcesAndVelocity(Clamp Velocity=false, Limit Acceleration=false)
   &ScaleColor(
    Scale RGB=&MakeVectorFromLinearColorRGB(Color=&VariantColorSelector(
      Variant Amount=0
      Original Color=&ColorFromCurve(ColorCurve=DataInterface() {propertyValues {
         AlphaCurve {keys [<at 0 1 RCIM_Cubic>,<at 0.9698 0>]}
         BlueCurve {keys [
           <at 0.0067 13.8025>
           <at 0.2903 1.8333>
           {arriveTangent=-2.6759, interpMode=RCIM_Cubic, leaveTangent=-2.6759, time=0.5749, value=0.275}
           <at 0.9701 0.0143>
          ]}
         GreenCurve {keys [
           <at 0.0067 20>
           <at 0.2903 5>
           {arriveTangent=-7.3257, interpMode=RCIM_Cubic, leaveTangent=-7.3257, time=0.5749, value=0.5}
           <at 0.9701 0.02>
          ]}
         RedCurve {keys [
           <at 0.0067 0.4289>
           <at 0.2903 0>
           {arriveTangent=0.0147, interpMode=RCIM_Cubic, leaveTangent=0.0147, time=0.5749, value=0}
           <at 0.9701 0.01>
          ]}
        }})
      Variant 1 Color=&ColorFromCurve(ColorCurve=DataInterface() {propertyValues {
         AlphaCurve {keys [<at 0 1 RCIM_Cubic>,<at 0.9698 0>]}
         BlueCurve {keys [
           <at 0.0067 13.8025>
           <at 0.2903 5>
           {arriveTangent=-7.3341, interpMode=RCIM_Cubic, leaveTangent=-7.3341, time=0.5749, value=0.5}
           <at 0.9701 0.0143>
          ]}
         GreenCurve {keys [
           <at 0.0067 20>
           <at 0.2903 0.1667>
           {arriveTangent=-0.2157, interpMode=RCIM_Cubic, leaveTangent=-0.2157, time=0.5749, value=0}
           <at 0.9701 0.02>
          ]}
         RedCurve {keys [
           <at 0.0067 0.4289>
           <at 0.2903 0>
           {arriveTangent=0.0147, interpMode=RCIM_Cubic, leaveTangent=0.0147, time=0.5749, value=0.0917}
           <at 0.9701 0.01>
          ]}
        }})
      Variant 2 Color=?
     ))
    ScaleA=true
    ScaleRGB=true
   )
  ]
  rendererClasses [&NiagaraSpriteRendererProperties]
  renderers [{rendererClass=&NiagaraSpriteRendererProperties}]
 }
 Flare {
  emitterUpdateScript [
   &EmitterState(
    Inactive Response=::NewEnumerator1
    Life Cycle Mode=::NewEnumerator1
    Loop Behavior=::NewEnumerator1
    Loop Delay=0.32
    Loop Duration=2.5
    UseLoopDelay=true
   )
   &SpawnBurst_Instantaneous(Spawn Group=1, Use Spawn Probability=false, UseLoopCountLimit=false) {module=SpawnBurst_Instantaneous001}
   &SpawnBurst_Instantaneous(Spawn Count=1, Spawn Group=2, Spawn Time=0.02, Use Spawn Probability=false) {enabled=false}
   &SpawnBurst_Instantaneous(Spawn Count=1, Spawn Group=2, Use Spawn Probability=false) {module=SpawnBurst_Instantaneous002}
  ]
  particleSpawnScript [
   /SetVariables(Particles.VisibilityTag=$Emitter.SpawnGroup) {isSetParameters=true}
   &InitializeParticle(
    Color=&MergeHelper_ColorSwitch(
     0=&MultiplyLinearColor(
      Linear Color=&VariantColorSelector(Variant Amount=0, Original Color=<rgba 0 0.7 0.3231 1>, Variant 1 Color=<rgba 0 0.035 0.7 1>, Variant 2 Color=?)
      Scale Factor=&MakeVector4_FromLinearColor(LinearColor=&FNTimeOfDayColor(Night Color=<rgba 1 1 1 1>, Morning Color=<rgba 1 1 1 1>, Day Color=<rgba 1 1 1 1>, Evening Color=<rgba 1 0.5625 0.5625 1>))
     )
     1=<rgba 0 0 0 0>
     2=&VariantColorSelector(Variant Amount=0, Original Color=<rgba 0 7 1.4 1>, Variant 1 Color=<rgba 0 1.5167 7 1>, Variant 2 Color=?)
     3=<rgba 0 0 0 0>
     4=<rgba 0 0 0 0>
     Index=$Particles.VisibilityTag
     Num=::NewEnumerator3
    )
    Lifetime=&MergeHelper_FloatSwitch(
     0=&Lerp_Float(A=0.4, B=0.65, Alpha=&MakeFloatFromInt(Int=$Emitter.SpawnGroup))
     1=0
     2=0.35
     3=0
     4=0
     Index=$Particles.VisibilityTag
     Num=::NewEnumerator3
    )
    Sprite Size=&MergeHelper_Vector2Switch(0=<350 400>, 1=<0 0>, 2=<40 600>, 3=?, 4=?, Index=$Particles.VisibilityTag, Num=::NewEnumerator1)
    Sprite Size Mode=::NewEnumerator4
    UsePositionOffset=true
    Write Material Random=true
   )
   &CameraOffset(Camera Offset Amount=50)
  ]
  particleUpdateScript [
   &ParticleState()
   &SkeletalMeshLocation(
    Bone Sampling Mode=::NewEnumerator4
    MaskSpawnGroup=false
    Sampled Position Offset=&AddVectorToPosition(
     Position=$User.DoorOffsets
     Vector=&Lerp_Vector(A=<0 0 0>, B=<0 10 0>, Alpha=&MakeFloatFromInt(Int=$Particles.VisibilityTag))
    )
    Skeletal Mesh=$System.Skeletal Mesh
   )
   &ScaleSpriteSize(
    Scale Factor=&MergeHelper_Vector2Switch(
     0=&Vector2DFromFloat(Value=&FloatFromCurve(FloatCurve=DataInterface() {propertyValues {Curve {keys [
           {arriveTangent=1.9515, interpMode=RCIM_Cubic, leaveTangent=1.9515, tangentMode=RCTM_User, time=0, value=0.2}
           {arriveTangent=2.2066, interpMode=RCIM_Cubic, leaveTangent=2.2066, tangentMode=RCTM_User, time=0.3, value=1}
           <at 0.7 3.5 RCIM_Cubic>
           {arriveTangent=0.0224, interpMode=RCIM_Cubic, leaveTangent=0.0224, tangentMode=RCTM_User, time=1, value=1}
          ]}}}))
     1=<0 0>
     2=&Vector2DFromCurve(
      Vector2Curve=DataInterface() {propertyValues {
        XCurve {keys [
          <at 0 0.4 RCIM_Cubic>
          {arriveTangent=9.4286, interpMode=RCIM_Cubic, leaveTangent=9.4286, time=0.4, value=2.5}
          {interpMode=RCIM_Cubic, leaveTangent=5.25, time=0.7, value=7}
         ]}
        YCurve {keys [
          {interpMode=RCIM_Cubic, tangentMode=RCTM_User, time=0, value=0.2}
          {interpMode=RCIM_Cubic, tangentMode=RCTM_User, time=0.7, value=4}
          {interpMode=RCIM_Cubic, tangentMode=RCTM_User, time=1, value=0}
         ]}
       }}
      ScaleCurve=<1 0.7>
     )
     3=<0 0>
     4=<0 0>
     Index=$Particles.VisibilityTag
     Num=::NewEnumerator3
    )
    Scale Sprite Size Mode=::NewEnumerator2
   )
   &ScaleColor(
    Scale Alpha=&FloatFromCurve(FloatCurve=DataInterface() {propertyValues {Curve {keys [{arriveTangent=-2.6688, interpMode=RCIM_Cubic, leaveTangent=-2.6688, tangentMode=RCTM_User, time=0, value=1}, <at 1 0>]}}})
    Scale RGB=&VectorFromFloat(Value=&FloatFromCurve(
      FloatCurve=DataInterface() {propertyValues {Curve {keys [{arriveTangent=-2.6688, interpMode=RCIM_Cubic, leaveTangent=-2.6688, tangentMode=RCTM_User, time=0, value=1}, <at 1 0>]}}}
      Scale Curve=$System.TOD
     ))
    ScaleA=true
    ScaleRGB=true
   )
   &FN_Sprite_Attributes(
    Color=&MultiplyLinearColor(
     Linear Color=<rgba 0 0.04 0.0073 0.7>
     Scale Factor=&MakeVector4(
      X=1
      Y=1
      Z=1
      W=&FloatFromCurve(FloatCurve=DataInterface() {propertyValues {Curve {keys [<at 0 1>,<at 1 0>]}}})
     )
    )
    Create Color Attribute=true
    Create SpriteSize Attribute=true
    SpriteSize=<600 600>
    Write to Module Namespace=true
   )
   &FN_Light_Attributes(
    Color=&VariantColorSelector(Variant Amount=0, Original Color=<rgba 0 0.25 0.1025 1>, Variant 1 Color=<rgba 0 0.1125 0.25 1>, Variant 2 Color=?)
    Create Color Attribute=true
    Create Exponent Attribute=true
    Create Radius Attribute=true
    Radius=200
    Scale Color=&FloatFromCurve(
     FloatCurve=DataInterface() {propertyValues {Curve {keys [
         <at 0 0 RCIM_Cubic>
         {arriveTangent=2.5, interpMode=RCIM_Cubic, leaveTangent=2.5, time=0.02, value=1}
         {interpMode=RCIM_Cubic, tangentMode=RCTM_User, time=0.4, value=1}
         {arriveTangent=-1.6168, interpMode=RCIM_Cubic, leaveTangent=-1.6168, tangentMode=RCTM_User, time=1, value=0}
        ]}}}
     Scale Curve=300
    )
   )
  ]
  rendererClasses [&NiagaraSpriteRendererProperties, &NiagaraLightRendererProperties]
  renderers [
   {rendererClass=&NiagaraSpriteRendererProperties}
   {rendererClass=&NiagaraSpriteRendererProperties, rendererIndex=1}
   {rendererClass=&NiagaraLightRendererProperties, rendererIndex=2}
   {rendererClass=&NiagaraSpriteRendererProperties, rendererIndex=3}
  ]
 }
 MainTrail {
  emitterUpdateScript [&EmitterState(
    Inactive Response=::NewEnumerator1
    Life Cycle Mode=::NewEnumerator1
    Loop Behavior=::NewEnumerator1
    Loop Delay=0.35
    Loop Duration=2
    UseLoopDelay=true
   )]
  particleSpawnScript [
   &InitializeParticle(
    Color Mode=::NewEnumerator0
    Lifetime=8
    Ribbon Width=60
    Ribbon Width Mode=::NewEnumerator1
    Sprite Size Mode=::NewEnumerator3
    Uniform Sprite Size=400
    UsePositionOffset=false
    Write Material Random=true
   )
   /SetVariables(Particles.MasterWidth:NiagaraFloat=&FloatFromCurve(
     FloatCurve=DataInterface() {propertyValues {Curve {keys [
         <at 0.5 0 RCIM_Cubic>
         {arriveTangent=3, interpMode=RCIM_Cubic, leaveTangent=3, time=0.8, value=1}
         {arriveTangent=-1.4, interpMode=RCIM_Cubic, leaveTangent=-1.4, time=1, value=1.5}
         {arriveTangent=-4.2857, interpMode=RCIM_Cubic, leaveTangent=-4.2857, time=1.3, value=0.3}
         {interpMode=RCIM_Cubic, leaveTangent=3.625, time=1.35, value=0}
        ]}}}
     CurveIndex=$Emitter.Age
    )) {isSetParameters=true}
   /SetVariables(Particles.Position=&AddVectorToPosition(Position=$Particles.Position, Vector=&TransformVector(Vector=<-100 0 0>, Destination Space=::World))) {isSetParameters=true}
  ]
  particleUpdateScript [
   &ParticleState()
   &SolveForcesAndVelocity(Clamp Velocity=false, Limit Acceleration=false)
   /SetVariables(
    Particles.Color=&VariantColorSelector(
     Variant Amount=0
     Original Color=&ColorFromCurve(
      ColorCurve=DataInterface() {propertyValues {
        AlphaCurve {keys [<at 0.1381 1 RCIM_Cubic>,<at 0.3648 0 RCIM_Cubic>]}
        BlueCurve {keys [<at 0.0487 1>, {interpMode=RCIM_Cubic, leaveTangent=-2.3331, time=0.1272, value=1.0831}]}
        GreenCurve {keys [<at 0.0487 4>, {interpMode=RCIM_Cubic, leaveTangent=-9.3761, time=0.1272, value=2}]}
        RedCurve {keys [<at 0.0487 0>, {interpMode=RCIM_Cubic, leaveTangent=0.004, time=0.1272, value=0.1705}]}
       }}
      CurveIndex=$Particles.Age
     )
     Variant 1 Color=&ColorFromCurve(
      ColorCurve=DataInterface() {propertyValues {
        AlphaCurve {keys [<at 0.1381 1 RCIM_Cubic>,<at 0.3648 0 RCIM_Cubic>]}
        BlueCurve {keys [<at 0.0487 40>, {interpMode=RCIM_Cubic, leaveTangent=-2.3331, time=0.1272, value=2}]}
        GreenCurve {keys [<at 0.0487 12.4008>, {interpMode=RCIM_Cubic, leaveTangent=-9.3761, time=0.1272, value=0.3839}]}
        RedCurve {keys [<at 0.0487 0>, {interpMode=RCIM_Cubic, leaveTangent=0.004, time=0.1272, value=0.1705}]}
       }}
      CurveIndex=$Particles.Age
     )
     Variant 2 Color=?
    )
    Particles.RibbonWidth=&Multiply_Float(A=60, B=$Particles.MasterWidth)
   ) {isSetParameters=true}
   /SetVariables(
    Particles.Ribbon2Color:LinearColor=&VariantColorSelector(
     Variant Amount=0
     Original Color=&ColorFromCurve(
      ColorCurve=DataInterface() {propertyValues {
        AlphaCurve {keys [<at 0.31 1 RCIM_Cubic>,<at 0.5631 0 RCIM_Cubic>]}
        BlueCurve {keys [<at 0.0487 1>, {arriveTangent=-1.9561, interpMode=RCIM_Cubic, leaveTangent=-1.9561, time=0.1272, value=0.3902}, <at 0.4732 0.1697>]}
        GreenCurve {keys [<at 0.0487 4>, {arriveTangent=-8.9521, interpMode=RCIM_Cubic, leaveTangent=-8.9521, time=0.1272, value=1}, <at 0.4732 0.2>]}
        RedCurve {keys [<at 0.0487 0>, {arriveTangent=0.1413, interpMode=RCIM_Cubic, leaveTangent=0.1413, time=0.1272, value=0.0852}, <at 0.4732 0.06>]}
       }}
      CurveIndex=$Particles.Age
     )
     Variant 1 Color=&ColorFromCurve(
      ColorCurve=DataInterface() {propertyValues {
        AlphaCurve {keys [<at 0.31 1>,<at 0.5631 0>]}
        BlueCurve {keys [<at 0.0487 4>,<at 0.1272 1>,<at 0.4732 0.15>]}
        GreenCurve {keys [<at 0.0487 2.4667>,<at 0.1272 0.1462>,<at 0.4732 0.075>]}
        RedCurve {keys [<at 0.0487 0>,<at 0.1272 0.0852>,<at 0.4732 0.0925>]}
       }}
      CurveIndex=$Particles.Age
     )
     Variant 2 Color=?
    )
    Particles.Ribbon2Width:NiagaraFloat=&Multiply_Float(A=60, B=$Particles.MasterWidth)
   ) {isSetParameters=true}
   /SetVariables(
    Particles.Ribbon3Color:LinearColor=&VariantColorSelector(
     Variant Amount=0
     Original Color=&ColorFromCurve(
      ColorCurve=DataInterface() {propertyValues {
        AlphaCurve {keys [<at 0.31 1 RCIM_Cubic>,<at 0.8548 0 RCIM_Cubic>]}
        BlueCurve {keys [<at 0.0487 1>, {arriveTangent=-2.1021, interpMode=RCIM_Cubic, leaveTangent=-2.1021, time=0.1272, value=0.5416}, <at 0.4562 0.1435>]}
        GreenCurve {keys [<at 0.0487 4>, {arriveTangent=-9.326, interpMode=RCIM_Cubic, leaveTangent=-9.326, time=0.1272, value=1}, <at 0.4562 0.2>]}
        RedCurve {keys [<at 0.0487 0>, {arriveTangent=0.1123, interpMode=RCIM_Cubic, leaveTangent=0.1123, time=0.1272, value=0.0852}, <at 0.4562 0.0458>]}
       }}
      CurveIndex=$Particles.Age
     )
     Variant 1 Color=&ColorFromCurve(
      ColorCurve=DataInterface() {propertyValues {
        AlphaCurve {keys [<at 0.31 1 RCIM_Cubic>,<at 0.8548 0 RCIM_Cubic>]}
        BlueCurve {keys [<at 0.0487 4>, {arriveTangent=-9.326, interpMode=RCIM_Cubic, leaveTangent=-9.326, time=0.1272, value=1}, <at 0.4562 0.2>]}
        GreenCurve {keys [<at 0.0487 0.6>, {arriveTangent=-1.2466, interpMode=RCIM_Cubic, leaveTangent=-1.2466, time=0.1272, value=0.4359}, <at 0.4562 0.092>]}
        RedCurve {keys [<at 0.0487 0>, {arriveTangent=0.1123, interpMode=RCIM_Cubic, leaveTangent=0.1123, time=0.1272, value=0.0852}, <at 0.4562 0.0458>]}
       }}
      CurveIndex=$Particles.Age
     )
     Variant 2 Color=?
    )
    Particles.Ribbon3Width:NiagaraFloat=&Multiply_Float(A=70, B=$Particles.MasterWidth)
   ) {isSetParameters=true}
   /SetVariables(
    Particles.Ribbon4Color:LinearColor=&MultiplyLinearColor(
     Linear Color=&ColorFromCurve(
      ColorCurve=DataInterface() {propertyValues {
        AlphaCurve {keys [{arriveTangent=-1.5048, interpMode=RCIM_Cubic, time=0.1603, value=0.7}, <at 0.6645 0 RCIM_Cubic>]}
        BlueCurve {keys [<at 0.0487 0>]}
        GreenCurve {keys [<at 0.0487 0>]}
        RedCurve {keys [<at 0.0487 0>]}
       }}
      CurveIndex=&Multiply_Float(A=$Emitter.Age, B=0.5)
     )
     Scale Factor=<1 1 1 1>
    )
    Particles.Ribbon4Width:NiagaraFloat=&Multiply_Float(
     A=&FloatFromCurve(FloatCurve=DataInterface() {propertyValues {Curve {keys [<at 0 0>,<at 0.02 1>]}}}, Scale Curve=100)
     B=$Particles.MasterWidth
    )
   ) {isSetParameters=true}
   &DynamicMaterialParameters(
    Index 0 Param 1=&FloatFromCurve(FloatCurve=DataInterface() {propertyValues {Curve {keys [<at 0 0>,<at 0.02 1>,<at 0.1 0>]}}})
    Module.Index0Param0Write=true
    Module.Index0Param1Write=true
    Module.Index0Param2Write=true
    Module.Index0Param3Write=true
   )
   &self:ScaleColorsByTOD(
    Linear Color:LinearColor=$Particles.Color
    Linear Color001:LinearColor=$Particles.Ribbon2Color
    Linear Color002:LinearColor=$Particles.Ribbon3Color
    Linear Color003:LinearColor=$Particles.Ribbon4Color
    TOD:NiagaraFloat=$System.TOD
   ) {module=ScaleColorsByTOD}
  ]
  rendererClasses [&NiagaraRibbonRendererProperties]
  renderers [
   {rendererClass=&NiagaraRibbonRendererProperties}
   {rendererClass=&NiagaraRibbonRendererProperties, rendererIndex=1}
   {rendererClass=&NiagaraRibbonRendererProperties, rendererIndex=2}
   {rendererClass=&NiagaraRibbonRendererProperties, rendererIndex=3}
  ]
 }
 PortalOpenSmoke {
  emitterUpdateScript [
   &EmitterState(
    Life Cycle Mode=::NewEnumerator1
    Loop Behavior=::NewEnumerator1
    Loop Delay=&Add_Float(A=$User.Beat2, B=-0.05)
    Loop Duration=1
    UseLoopDelay=true
   )
   &SpawnBurst_Instantaneous(Spawn Count=2, Use Spawn Probability=false, UseLoopCountLimit=false)
  ]
  particleSpawnScript [
   &PartitionParticles()
   &InitializeParticle(
    Color=&MergeHelper_ColorSwitch(
     0=&MultiplyLinearColor(
      Linear Color=&VariantColorSelector(Variant Amount=0, Original Color=<rgba 0 8 2.2667 1>, Variant 1 Color=<rgba 0 0.8 8 1>, Variant 2 Color=?)
      Scale Factor=&MakeVector4_FromLinearColor(LinearColor=&FNTimeOfDayColor(Night Color=<rgba 1 1 1 1>, Morning Color=<rgba 1 1 1 1>, Day Color=<rgba 1 1 1 1>, Evening Color=<rgba 0.5 0.4027 0.3253 0.4>))
     )
     1=&VariantColorSelector(Variant Amount=0, Original Color=<rgba 0 0 0 3>, Variant 1 Color=<rgba 0.1 0.1 0.1 3>, Variant 2 Color=?)
     2=?
     3=?
     Index=$Particles.PartitionParticles.Partition
     Num=::NewEnumerator0
    )
    Lifetime=&MergeHelper_FloatSwitch(0=0.35, 1=0.7, 2=?, 3=?, Index=$Particles.PartitionParticles.Partition, Num=::NewEnumerator0)
    Sprite Rotation Mode=::NewEnumerator2
    Sprite Size=<200 150>
    Sprite Size Mode=::NewEnumerator4
    Sprite UV Mode=::NewEnumerator3
    UsePositionOffset=true
    Write Material Random=true
   )
   &SkeletalMeshLocation(
    Bone Sampling Mode=::NewEnumerator4
    MaskSpawnGroup=false
    Sampled Position Offset=&AddVectorToPosition(
     Position=$User.DoorOffsets
     Vector=&MergeHelper_VectorSwitch(0=<0 0 0>, 1=<0 0 20>, 2=?, 3=?, Index=$Particles.PartitionParticles.Partition, Num=::NewEnumerator0)
    )
    Skeletal Mesh=$System.Skeletal Mesh
   )
   &AddVelocity(
    Velocity=&MergeHelper_VectorSwitch(0=<0 0 20>, 1=<0 0 0>, 2=?, 3=?, Index=$Particles.PartitionParticles.Partition, Num=::NewEnumerator0)
    Velocity Speed Scale=2
   )
   &ShapeLocation(
    Handle Radius=0
    Large Radius=10
    MaskSpawnGroup=false
    Override Seed=false
    Shape Primitive=::NewEnumerator3
    Surface Distribution=1
    UseOwnerScale=true
   )
   &CameraOffset(Camera Offset Amount=&MergeHelper_FloatSwitch(0=30, 1=0, 2=?, 3=?, Index=$Particles.PartitionParticles.Partition, Num=::NewEnumerator0)) {module=CameraOffset001}
  ]
  particleUpdateScript [
   &ParticleState()
   &SpriteRotationRate(Rotation Rate=&RandomRangeFloat(Minimum=-150, Maximum=150))
   &Drag(Drag=3, Rotational Drag=5, Use Linear Drag=true, Use Rotational Drag=true)
   &SolveForcesAndVelocity(Clamp Velocity=false, Limit Acceleration=false)
   &ScaleSpriteSizeBySpeed(Max Scale Factor=<1 5>, Min Scale Factor=<1 1.5>)
   &ScaleColor(
    Scale Alpha=&FloatFromCurve(FloatCurve=DataInterface() {propertyValues {Curve {keys [
         {interpMode=RCIM_Cubic, tangentMode=RCTM_User, time=0, value=0}
         {arriveTangent=-2.0189, interpMode=RCIM_Cubic, leaveTangent=-2.1015, tangentMode=RCTM_Break, time=0.2, value=0.9}
         <at 1 0 RCIM_Cubic>
        ]}}})
    ScaleA=true
    ScaleRGB=true
   )
  ]
  rendererClasses [&NiagaraSpriteRendererProperties]
  renderers [
   {rendererClass=&NiagaraSpriteRendererProperties}
   {rendererClass=&NiagaraSpriteRendererProperties, rendererIndex=1}
  ]
 }
 Smoke_DriftOut {
  emitterUpdateScript [
   &EmitterState(Life Cycle Mode=::NewEnumerator1, Loop Behavior=::NewEnumerator1, Loop Delay=1.1, Loop Duration=0.3, UseLoopDelay=true)
   &SpawnRate(SpawnRate=100, Use Spawn Probability=false)
   &SpawnBurst_Instantaneous(Spawn Count=30, Use Spawn Probability=false, UseLoopCountLimit=false) {enabled=false}
  ]
  particleSpawnScript [
   /SetVariables(Particles.modulator:NiagaraFloat=&UniformAOrBFloat(A=0, Distribution Weight=1)) {isSetParameters=true, module=SetVariables_#4}
   &InitializeParticle(
    Color=<rgba 1 1 1 0.2>
    Lifetime=&Lerp_Float(
     A=&RandomRangeFloat(Minimum=0.1, Maximum=0.3)
     B=&RandomRangeFloat(Minimum=0.3, Maximum=0.8)
     Alpha=$Particles.modulator
    )
    Sprite Rotation Mode=::NewEnumerator2
    Sprite Size Mode=::NewEnumerator3
    Uniform Sprite Size=100
    UsePositionOffset=false
    Write Material Random=true
   )
   &SkeletalMeshLocation(
    MaskSpawnGroup=false
    Mesh Sampling Type=::NewEnumerator5
    Sampled Position Offset=<0 0 -20>
    Sampled Position Offset Coordinate Space:ENiagaraSkelMeshTransforms=::NewEnumerator2
    Skeletal Mesh=$System.Skeletal Mesh
    Socket Sampling Mode=::NewEnumerator5
    Velocity Sampling=::NewEnumerator0
    VelocitySpeedLimit=true
   )
   &ShapeLocation(MaskSpawnGroup=false, Override Seed=false, Sphere Radius=5, UseOwnerScale=true)
   &AddVelocity(Velocity=<0 0 150>, Velocity Speed Scale=1.5)
   &BoxLocation(Box Size=<20 1 1>, MaskSpawnGroup=false, Module.Override Seed=false, Rotate Box=false, Surface Only=false) {enabled=false}
   &ApplyInitialForces() {enabled=false}
  ]
  particleUpdateScript [
   &ParticleState()
   &ScaleColor(
    Scale Mode=::NewEnumerator1
    Scale RGBA=&MakeVector4_FromLinearColor(LinearColor=&MultiplyLinearColor(
      Linear Color=&VariantColorSelector(
       Variant Amount=0
       Morning Color=<rgba 1 1 1 1>
       Day Color=<rgba 1 1 1 1>
       Evening Color=<rgba 1 0.4853 0.7936 1>
       Night Color=<rgba 0.3385 0.7 0.4486 1>
       Use TOD?=true
       Original Color=&ColorFromCurve(ColorCurve=DataInterface() {propertyValues {
          AlphaCurve {keys [<at 0 1 RCIM_Cubic>,<at 0.3879 1>,<at 1 0 RCIM_Cubic>]}
          BlueCurve {keys [<at 0 1.3333>,<at 0.0743 0.1329>,<at 0.3546 0.01 RCIM_Cubic>]}
          GreenCurve {keys [<at 0 4>,<at 0.0743 0.2542>,<at 0.3546 0.0097 RCIM_Cubic>]}
          RedCurve {keys [<at 0 0>,<at 0.0743 0.0116>,<at 0.3546 0.0076 RCIM_Cubic>]}
         }})
       Variant 1 Color=&ColorFromCurve(ColorCurve=DataInterface() {propertyValues {
          AlphaCurve {keys [<at 0 1 RCIM_Cubic>,<at 0.3879 1>,<at 1 0 RCIM_Cubic>]}
          BlueCurve {keys [<at 0 4>,<at 0.0743 0.2542>,<at 0.3546 0.01 RCIM_Cubic>]}
          GreenCurve {keys [<at 0 2.1333>,<at 0.0743 0.0925>,<at 0.3546 0.0097 RCIM_Cubic>]}
          RedCurve {keys [<at 0 0>,<at 0.0743 0.0116>,<at 0.3546 0.0076 RCIM_Cubic>]}
         }})
       Variant 2 Color=?
      )
      Scale Factor=&MakeFloat4FromVectorAndFloat("Vector (RGB)"=&VectorFromFloat(Value=5), "Float (Alpha)"=1)
     ))
    ScaleRGBA=true
   )
   &ScaleSpriteSize(Scale Factor=&Vector2DFromFloat(Value=&FloatFromCurve(
      FloatCurve=DataInterface() {propertyValues {Curve {keys [
          <at 0 0 RCIM_Cubic>
          {interpMode=RCIM_Cubic, leaveTangent=-3.0296, tangentMode=RCTM_Break, time=0.2, value=1.2}
          {arriveTangent=-1.8107, interpMode=RCIM_Cubic, leaveTangent=-1.8107, tangentMode=RCTM_User, time=1, value=0}
         ]}}}
      Scale Curve=0.4
     ))) {module=ScaleSpriteSize001}
   &LinearForce(Force=<-400 0 0>)
   &GravityForce(Gravity=<0 0 -200>)
   &CurlNoiseForce(
    Noise Frequency=300
    Noise Strength=&Lerp_Float(A=1000, B=1000, Alpha=$System.isLastSwing)
    Pan Noise=true
    Pan Noise Field=<0 0 1>
    Random Seed=10
   )
   &Drag(Drag=&RandomRangeFloat(Minimum=2, Maximum=5), Use Linear Drag=true, Use Rotational Drag=false)
   &SolveForcesAndVelocity(Clamp Velocity=false, Limit Acceleration=false)
  ]
  rendererClasses [&NiagaraSpriteRendererProperties]
  renderers [{rendererClass=&NiagaraSpriteRendererProperties}]
 }
 Smoke_In_out {
  emitterUpdateScript [
   &EmitterState(Life Cycle Mode=::NewEnumerator1, Loop Behavior=::NewEnumerator1, Loop Delay=0.08, Loop Duration=1, UseLoopDelay=true)
   &SpawnBurst_Instantaneous(Spawn Count=3, Spawn Time=0.1, Use Spawn Probability=false) {module=SpawnBurst_Instantaneous001}
   &SpawnBurst_Instantaneous(Spawn Count=7, Spawn Group=1, Spawn Time=0.9, Use Spawn Probability=false) {module=SpawnBurst_Instantaneous002}
   &SpawnBurst_Instantaneous(Spawn Count=7, Spawn Time=0.1, Use Spawn Probability=false) {enabled=false}
  ]
  particleSpawnScript [
   &InitializeParticle(
    Color=<rgba 0 0 0 1>
    Lifetime=&UniformRangedFloat(Minimum=0.2, Maximum=0.6)
    Sprite Rotation Angle=&UniformRangedFloat(Minimum=0, Maximum=1)
    Sprite Rotation Mode=::NewEnumerator1
    Sprite Size Mode=::NewEnumerator3
    Uniform Sprite Size=&UniformRangedFloat(Minimum=25, Maximum=35)
    UsePositionOffset=false
    Write Material Random=true
   ) {module=InitializeParticle001}
   &SkeletalMeshLocation(
    Bone Sampling Mode=::NewEnumerator4
    MaskSpawnGroup=false
    Sampled Position Offset=&AddVectorToPosition(Position=$User.DoorOffsets, Vector=<0 0 -20>)
    Skeletal Mesh=$System.Skeletal Mesh
   )
   /SetVariables(Particles.InitSpriteRotRate=&UniformRangedFloat(Minimum=-1, Maximum=1)) {isSetParameters=true}
   &AccelerationForce(Acceleration=&MergeHelper_VectorSwitch(
     0=&UniformRangedVector(Minimum=<0 0 50>, Maximum=<0 0 100>)
     1=&UniformRangedVector(Minimum=<0 0 300>, Maximum=<0 0 200>)
     2=?
     3=?
     Index=$Emitter.SpawnGroup
     Num=::NewEnumerator0
    ))
   &SphereLocation(
    MaskSpawnGroup=false
    Override Seed=false
    Sphere Radius=&MergeHelper_FloatSwitch(0=10, 1=20, 2=?, 3=?, Index=$Emitter.SpawnGroup, Num=::NewEnumerator0)
    Surface Only=true
    UseNonUniformScale=false
    UseOffset=true
    UseSphereOrientation=false
   )
   &ApplyInitialForces()
   &CameraOffset(Camera Offset Amount=&RandomRangeFloat(Minimum=0, Maximum=15))
   &AddVelocity(Velocity=&MergeHelper_VectorSwitch(
     0=&Multiply_Vector(
      A=$Output.SphereLocation.SphereNormal
      B=&VectorFromFloat(Value=&RandomRangeFloat(Minimum=20, Maximum=150))
     )
     1=<0 0 0>
     2=?
     3=?
     Index=$Emitter.SpawnGroup
     Num=::NewEnumerator0
    ))
   &AddVelocity(
    Velocity=&MergeHelper_VectorSwitch(
     0=<0 0 0>
     1=&RandomRangeVector(Minimum=<400 100 0>, Maximum=<-100 -100 0>)
     2=?
     3=?
     Index=$Emitter.SpawnGroup
     Num=::NewEnumerator0
    )
    Velocity Speed Scale=2
   ) {module=AddVelocity001}
  ]
  particleUpdateScript [
   &ParticleState()
   &SpriteRotationRate(Rotation Rate=&Multiply_Float(
     A=$Particles.InitSpriteRotRate
     B=&FloatFromCurve(
      FloatCurve=DataInterface() {propertyValues {Curve {keys [
          {arriveTangent=-1.9214, interpMode=RCIM_Cubic, leaveTangent=-1.9214, tangentMode=RCTM_User, time=-0.0046, value=1.0113}
          {arriveTangent=-0.2496, interpMode=RCIM_Cubic, leaveTangent=-0.2496, tangentMode=RCTM_User, time=0.9896, value=0.16}
         ]}}}
      CurveIndex=$Emitter.Age
     )
    ))
   &ScaleColor(
    Scale Alpha=&FloatFromCurve(FloatCurve=DataInterface() {propertyValues {Curve {keys [<at 0 0>,<at 0.2049 0.9765>,<at 0.7791 0.9107>,<at 1 0>]}}})
    ScaleA=true
    ScaleRGB=true
   ) {module=ScaleColor001}
   &Drag(Drag=3, Use Linear Drag=true, Use Rotational Drag=false)
   &SolveForcesAndVelocity(Clamp Velocity=false, Limit Acceleration=false)
   &ScaleSpriteSize(Scale Factor=&Vector2DFromFloat(Value=&FloatFromCurve(FloatCurve=DataInterface() {propertyValues {Curve {keys [
          {arriveTangent=5.2354, interpMode=RCIM_Cubic, leaveTangent=5.2354, tangentMode=RCTM_User, time=-0.0046, value=0.25}
          <at 0.1439 1.4923>
          <at 1.0011 3.4809 RCIM_Cubic>
         ]}}})))
   &DynamicMaterialParameters(
    Index 0 Param 1=&FloatFromCurve(FloatCurve=DataInterface() {propertyValues {Curve {keys [<at 0 0 RCIM_Cubic>,<at 1.0011 0.2>]}}})
    Index 0 Param 2=&FloatFromCurve(FloatCurve=DataInterface() {propertyValues {Curve {keys [
         {arriveTangent=0.0453, interpMode=RCIM_Cubic, tangentMode=RCTM_Break, time=0, value=-1}
         {arriveTangent=0.0604, interpMode=RCIM_Cubic, leaveTangent=0.0604, tangentMode=RCTM_User, time=1, value=-1}
        ]}}})
    Index 0 Param 3=&FloatFromCurve(FloatCurve=DataInterface() {propertyValues {Curve {keys [{interpMode=RCIM_Cubic, leaveTangent=4.6779, time=0.2851, value=0}]}}})
    Index 1 Param 1=&FloatFromCurve(
     FloatCurve=DataInterface() {propertyValues {Curve {keys [<at 0 0>,<at 0.25 1>]}}}
     Scale Curve=&RandomRangeFloat(Minimum=2, Maximum=1)
    )
    Index 1 Param 2=&FloatFromCurve(
     FloatCurve=DataInterface() {propertyValues {Curve {keys [
         {interpMode=RCIM_Cubic, leaveTangent=3.7422, tangentMode=RCTM_Break, time=0, value=0.2}
         {interpMode=RCIM_Cubic, leaveTangent=0.9066, time=1, value=2}
        ]}}}
     Scale Curve=&RandomRangeFloat(Minimum=0.5, Maximum=1)
    )
    Index 1 Param 3=&FloatFromCurve(FloatCurve=DataInterface() {propertyValues {Curve {keys [
         {interpMode=RCIM_Cubic, leaveTangent=2.8445, tangentMode=RCTM_Break, time=0, value=0.2}
         {interpMode=RCIM_Cubic, leaveTangent=63.2978, time=0.25, value=1}
        ]}}})
    Index 1 Param 4=&FloatFromCurve(
     FloatCurve=DataInterface() {propertyValues {Curve {keys [<at 0 1>,<at 1 0.8>]}}}
     Scale Curve=&RandomRangeFloat(Minimum=0.5, Maximum=1)
    )
    Index 2 Param 1=&FloatFromCurve(FloatCurve=DataInterface() {propertyValues {Curve {keys [{interpMode=RCIM_Cubic, leaveTangent=3.2309, tangentMode=RCTM_Break, time=0.1, value=0}, <at 1 1>]}}})
    Index 2 Param 2=$Particles.MaterialRandom
    Index 2 Param 3=&FloatFromCurve(FloatCurve=DataInterface() {propertyValues {Curve {keys [
         {interpMode=RCIM_Cubic, tangentMode=RCTM_Break, time=0, value=1}
         {interpMode=RCIM_Cubic, leaveTangent=11.1233, tangentMode=RCTM_Break, time=1, value=2}
        ]}}})
    Index 2 Param 4=0.75
    Module.Param0WriteEnabled=true
    Module.Param0WriteEnabled1=true
    Module.Param0WriteEnabled2=true
    Module.Param1WriteEnabled=false
    Module.Param1WriteEnabled1=true
    Module.Param1WriteEnabled2=true
    Module.Param2WriteEnabled=true
    Module.Param2WriteEnabled1=true
    Module.Param2WriteEnabled2=true
    Module.Param3WriteEnabled=true
    Module.Param3WriteEnabled1=true
    Module.Param3WriteEnabled2=true
    Write Parameter Index 1=true
    Write Parameter Index 2=true
   )
  ]
  rendererClasses [&NiagaraSpriteRendererProperties]
  renderers [{rendererClass=&NiagaraSpriteRendererProperties}]
 }
 Sparks {
  simTarget=GPUComputeSim
  emitterUpdateScript [
   &EmitterState(
    Life Cycle Mode=::NewEnumerator1
    Loop Behavior=::NewEnumerator1
    Loop Delay=&Add_Float(A=$User.Beat2, B=0)
    Loop Duration=1.7
    UseLoopDelay=true
   )
   &SpawnBurst_Instantaneous(Spawn Count=20, Use Spawn Probability=false, UseLoopCountLimit=false)
  ]
  particleSpawnScript [
   &InitializeParticle(
    Color=&VariantColorSelector(Variant Amount=0, Original Color=<rgba 0.1129 1 0.6895 1>, Variant 1 Color=<rgba 0.1129 0.4973 1 1>, Variant 2 Color=?)
    Lifetime Max=1
    Lifetime Min=0.4
    Lifetime Mode=::NewEnumerator1
    Position Offset=&MakeVector(
     X=0
     Y=0
     Z=&FloatFromCurve(FloatCurve=DataInterface() {propertyValues {Curve {keys [<at 0 0>,<at 1 1>]}}}, CurveIndex=$Emitter.NormalizedLoopAge, Scale Curve=140)
    )
    Sprite Size Max=<0.5 4>
    Sprite Size Min=<0.5 3>
    Sprite Size Mode=::NewEnumerator2
    UsePositionOffset=true
    Write Material Random=true
   )
   &SkeletalMeshLocation(
    Bone Sampling Mode=::NewEnumerator4
    MaskSpawnGroup=false
    Sampled Position Offset:NiagaraPosition=$User.DoorOffsets
    Skeletal Mesh=$System.Skeletal Mesh
   )
   &ShapeLocation(MaskSpawnGroup=false, Override Seed=false, Sphere Radius=50, UseOwnerScale=true)
   &AddVelocity(
    Cone Angle=8
    Cone Axis=<0 1 0>
    Module.UseConeFalloff=false
    Module.UseVelDistribution=false
    Override Seed=false
    Velocity Mode=::NewEnumerator2
    Velocity Speed=&Multiply_Float(A=4, B=&RandomRangeFloat(Minimum=300, Maximum=600))
   )
   &AddVelocity(Module.Use Falloff Distance=false, Origin Offset=<0 0 130>, Override Seed=false, Velocity Mode=::NewEnumerator1, Velocity Speed=50) {module=AddVelocity002}
  ]
  particleUpdateScript [
   &ParticleState()
   /SetVariables(Particles.Glow03_Color=&MakeLinearColorFromVectorAndFloat(
     "Vector (RGB)"=&MakeVectorFromLinearColorRGB(Color=&MultiplyLinearColor(Linear Color=<rgba 0 1 0.7 1>, Scale Factor=&MakeVector4_FromFloat(Float=1)))
     "Float (Alpha)"=&Multiply_Float(A=0.1, B=1)
    )) {isSetParameters=true, module=SetVariables_#5}
   /SetVariables(Particles.Glow03_Size=&Multiply_Vector2D_ByFloat(
     Float=&FloatFromCurve(
      FloatCurve=DataInterface() {propertyValues {Curve {keys [
          <at 0 0 RCIM_Cubic>
          <at 0.2 1.3 RCIM_Cubic>
          {arriveTangent=-0.1423, interpMode=RCIM_Cubic, leaveTangent=-0.1423, tangentMode=RCTM_User, time=1, value=0}
         ]}}}
      Scale Curve=10
     )
     Vector2D=<1 2>
    )) {isSetParameters=true, module=SetVariables_#6}
   &ScaleColor(
    Scale Alpha=&FloatFromCurve(FloatCurve=DataInterface() {propertyValues {Curve {keys [<at 0 1>,<at 0.3 1>,<at 1 1>]}}})
    ScaleA=true
    ScaleRGB=true
   )
   &ScaleSpriteSize(Scale Factor=&Vector2DFromFloat(Value=&FloatFromCurve(
      FloatCurve=DataInterface() {propertyValues {Curve {keys [
          <at 0 0 RCIM_Cubic>
          {interpMode=RCIM_Cubic, leaveTangent=-3.0296, tangentMode=RCTM_Break, time=0.2, value=1.2}
          {arriveTangent=-1.8107, interpMode=RCIM_Cubic, leaveTangent=-1.8107, tangentMode=RCTM_User, time=1, value=0}
         ]}}}
      Scale Curve=2
     ))) {module=ScaleSpriteSize001}
   &Drag(Drag=7, Rotational Drag=10, Use Linear Drag=true, Use Rotational Drag=false)
   &CurlNoiseForce(Noise Strength=1000, Pan Noise=false)
   &GravityForce(Gravity=&RandomRangeVector(Minimum=<0 0 20>, Maximum=<0 0 10>))
   &SolveForcesAndVelocity(Clamp Velocity=false, Limit Acceleration=false)
   &ScaleSpriteSizeBySpeed(Max Scale Factor=<1 3>)
  ]
  rendererClasses [&NiagaraSpriteRendererProperties]
  renderers [
   {rendererClass=&NiagaraSpriteRendererProperties}
   {rendererClass=&NiagaraSpriteRendererProperties, rendererIndex=1}
  ]
 }
 Sparks001 {
  simTarget=GPUComputeSim
  emitterUpdateScript [
   &EmitterState(Life Cycle Mode=::NewEnumerator1, Loop Behavior=::NewEnumerator1, Loop Delay=0.4, Loop Duration=1, UseLoopDelay=true)
   &SpawnRate(SpawnRate=400, Use Spawn Probability=false)
  ]
  particleSpawnScript [
   /SetVariables(Particles.modulator:NiagaraFloat=&UniformAOrBFloat(A=0, Distribution Weight=1)) {isSetParameters=true, module=SetVariables_#4}
   &InitializeParticle(
    Color=&VariantColorSelector(Variant Amount=0, Original Color=<rgba 0 4 3 0.5>, Variant 1 Color=<rgba 0 1.3333 4 0.5>, Variant 2 Color=?)
    Lifetime=&Lerp_Float(
     A=&RandomRangeFloat(Minimum=0.2, Maximum=0.4)
     B=&RandomRangeFloat(Minimum=0.1, Maximum=0.8)
     Alpha=$Particles.modulator
    )
    Sprite Size Max=<1 15>
    Sprite Size Min=<2 5>
    Sprite Size Mode=::NewEnumerator2
    UsePositionOffset=false
    Write Material Random=true
   )
   &SkeletalMeshLocation(
    "Inherit Velocity (Scale)"=4
    Inherited Velocity Speed Limit=200
    MaskSpawnGroup=false
    Mesh Sampling Type=::NewEnumerator5
    Sampled Position Offset=<0 -50 0>
    Sampled Position Offset Coordinate Space:ENiagaraSkelMeshTransforms=::NewEnumerator2
    Skeletal Mesh=$System.Skeletal Mesh
    Socket Sampling Mode=::NewEnumerator5
    Velocity Sampling=::NewEnumerator0
    VelocitySpeedLimit=true
   )
   &ShapeLocation(MaskSpawnGroup=false, Override Seed=false, Sphere Radius=20, UseOwnerScale=true)
   &AddVelocity(Velocity=<0 0 0>, Velocity Speed Scale=2)
   &ApplyInitialForces() {enabled=false}
  ]
  particleUpdateScript [
   &ParticleState()
   &ScaleColor(
    Scale Mode=::NewEnumerator1
    Scale RGBA=&MakeVector4_FromLinearColor(LinearColor=&MultiplyLinearColor(
      Linear Color=&ColorFromCurve(ColorCurve=DataInterface() {propertyValues {
         AlphaCurve {keys [<at 0 1 RCIM_Cubic>,<at 0.3879 1>,<at 1 0 RCIM_Cubic>]}
         BlueCurve {keys [<at 0 0.1329>,<at 0.3998 0.0545>,<at 0.6142 0.01 RCIM_Cubic>]}
         GreenCurve {keys [<at 0 0.2542>,<at 0.3998 0.0887>,<at 0.6142 0.0097 RCIM_Cubic>]}
         RedCurve {keys [<at 0 0.0116>,<at 0.3998 0.011>,<at 0.6142 0.0076 RCIM_Cubic>]}
        }})
      Scale Factor=&MakeFloat4FromVectorAndFloat("Vector (RGB)"=&VectorFromFloat(Value=5), "Float (Alpha)"=1)
     ))
    ScaleRGBA=true
   )
   /SetVariables(Particles.Glow03_Color=&MultiplyLinearColor(
     Linear Color=$Particles.Color
     Scale Factor=&MakeFloat4FromVectorAndFloat("Vector (RGB)"=&VectorFromFloat(Value=0.3), "Float (Alpha)"=1)
    )) {isSetParameters=true, module=SetVariables_#5}
   /SetVariables(Particles.Glow03_Size=&Multiply_Vector2D_ByFloat(
     Float=&FloatFromCurve(
      FloatCurve=DataInterface() {propertyValues {Curve {keys [
          <at 0 0 RCIM_Cubic>
          <at 0.2 1.3 RCIM_Cubic>
          {arriveTangent=-0.1423, interpMode=RCIM_Cubic, leaveTangent=-0.1423, tangentMode=RCTM_User, time=1, value=0}
         ]}}}
      Scale Curve=15
     )
     Vector2D=<0.5 1>
    )) {isSetParameters=true, module=SetVariables_#6}
   /SetVariables(Particles.SpriteAttributes.Glow03_Size:Vector2f=&Multiply_Vector2D_ByFloat(
     Float=&FloatFromCurve(
      FloatCurve=DataInterface() {propertyValues {Curve {keys [
          <at 0 0 RCIM_Cubic>
          <at 0.2 1.3 RCIM_Cubic>
          {arriveTangent=-0.1423, interpMode=RCIM_Cubic, leaveTangent=-0.1423, tangentMode=RCTM_User, time=1, value=0}
         ]}}}
      Scale Curve=15
     )
     Vector2D=<0.5 1>
    )) {isSetParameters=true}
   &ScaleColor(Scale RGB=&VectorFromFloat(Value=4), ScaleA=true, ScaleRGB=true) {module=ScaleColor001}
   &ScaleSpriteSize(Scale Factor=&Vector2DFromFloat(Value=&FloatFromCurve(
      FloatCurve=DataInterface() {propertyValues {Curve {keys [
          <at 0 0 RCIM_Cubic>
          {interpMode=RCIM_Cubic, leaveTangent=-3.0296, tangentMode=RCTM_Break, time=0.2, value=1.2}
          {arriveTangent=-1.8107, interpMode=RCIM_Cubic, leaveTangent=-1.8107, tangentMode=RCTM_User, time=1, value=0}
         ]}}}
      Scale Curve=0.4
     ))) {module=ScaleSpriteSize001}
   &LinearForce(Force=<-400 0 0>)
   &GravityForce(Gravity=&Lerp_Vector(A=<0 0 -200>, B=<0 0 -50>, Alpha=$System.isLastSwing))
   &CurlNoiseForce(
    Noise Frequency=300
    Noise Strength=&Lerp_Float(A=1000, B=1000, Alpha=$System.isLastSwing)
    Pan Noise=true
    Pan Noise Field=<0 0 1>
    Random Seed=10
   )
   &Drag(Drag=9, Use Linear Drag=true, Use Rotational Drag=false)
   &SolveForcesAndVelocity(Clamp Velocity=false, Limit Acceleration=false)
  ]
  rendererClasses [&NiagaraSpriteRendererProperties]
  renderers [
   {rendererClass=&NiagaraSpriteRendererProperties}
   {rendererClass=&NiagaraSpriteRendererProperties, rendererIndex=1}
  ]
 }
 TrailLeader {
  emitterUpdateScript [
   &EmitterState(
    Inactive Response=::NewEnumerator1
    Life Cycle Mode=::NewEnumerator1
    Loop Behavior=::NewEnumerator1
    Loop Delay=0.145
    Loop Duration=1.6
    UseLoopDelay=true
   )
   &SpawnBurst_Instantaneous(Spawn Count=1, Use Spawn Probability=false)
  ]
  particleSpawnScript [
   &InitializeParticle(
    Color=<rgba 0 7 5.3667 1>
    Lifetime=4
    Sprite Size=<124.2746 3>
    Sprite Size Mode=::NewEnumerator4
    UsePositionOffset=true
    Write Material Random=true
   )
   &SkeletalMeshLocation(MaskSpawnGroup=false, Mesh Sampling Type=::NewEnumerator5, Skeletal Mesh=$System.Skeletal Mesh, Socket Sampling Mode=::NewEnumerator5) {module=SkeletalMeshLocation001}
   &CameraOffset(Camera Offset Amount=40)
  ]
  particleUpdateScript [
   &SolveForcesAndVelocity(Clamp Velocity=false, Limit Acceleration=false)
   &ParticleState()
   &SkeletalMeshLocation(MaskSpawnGroup=false, Mesh Sampling Type=::NewEnumerator5, Skeletal Mesh=$System.Skeletal Mesh, Socket Sampling Mode=::NewEnumerator5)
   &ScaleSpriteSize(
    Uniform Curve Index=$Emitter.Age
    Uniform Curve Sprite Scale=DataInterface() {propertyValues {Curve {keys [{tangentMode=RCTM_User, time=0.5, value=0}, {tangentMode=RCTM_User, time=1, value=5}, <at 1.2 0>]}}}
   ) {module=ScaleSpriteSize002}
   /SetVariables(Particles.DistanceTraveled:NiagaraFloat=&DistanceBetweenPositions(A=$Particles.Initial.Position, B=$Particles.Position)) {isSetParameters=true, module=SetVariables_#3}
   &GenerateLocationEvent(Event Send Rate=100, Use Event Probability=false, UseEventDelay=false)
  ]
  rendererClasses []
  renderers []
 }
}
summary {
 systemName=NS_Glider_Gatehound_Deploy
 emitters [
  {emitterName=MainTrail, rendererClasses [&NiagaraRibbonRendererProperties]}
  {emitterName=Flare, rendererClasses [&NiagaraSpriteRendererProperties, &NiagaraLightRendererProperties]}
  {emitterName=TrailLeader, rendererClasses []}
  {emitterName=Sparks, simTarget=GPUComputeSim, rendererClasses [&NiagaraSpriteRendererProperties]}
  {emitterName=Smoke_In_out, rendererClasses [&NiagaraSpriteRendererProperties]}
  {emitterName=Bike_Flare, rendererClasses [&NiagaraSpriteRendererProperties]}
  {emitterName=PortalOpenSmoke, rendererClasses [&NiagaraSpriteRendererProperties]}
  {emitterName=Bike_Smoke, rendererClasses [&NiagaraSpriteRendererProperties]}
  {emitterName=Sparks001, simTarget=GPUComputeSim, rendererClasses [&NiagaraSpriteRendererProperties]}
  {emitterName=Smoke_DriftOut, rendererClasses [&NiagaraSpriteRendererProperties]}
  {emitterName=Flames, rendererClasses [&NiagaraSpriteRendererProperties]}
  {emitterName=Doors, rendererClasses [&NiagaraMeshRendererProperties, &NiagaraSpriteRendererProperties]}
 ]
 userVariables [
  User.Beat1 {type=NiagaraFloat, defaultValue=0.2, description=""}
  User.Beat2 {type=NiagaraFloat, defaultValue=0.5, description=""}
  User.Beat3 {type=NiagaraFloat, defaultValue=0.95, description=""}
  User.bIsFrontEnd {type=NiagaraBool}
  User.bIsFrontEndPreview {type=NiagaraBool}
  User.DoorOffsets {type=NiagaraPosition, description="", defaultValue {}}
  User.StartPos {type=NiagaraPosition, description="", defaultValue {}}
  User.Variant {type=NiagaraFloat, defaultValue=0, description=Choose Variant}
 ]
}
